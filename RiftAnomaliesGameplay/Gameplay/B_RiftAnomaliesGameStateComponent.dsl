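; Game state component that runs the Rift Anomalies system, spawning anomalies tied to storm circle phases.
; Data Registries supply the anomaly pool and config, with weighted selection shaped by subset overrides, modifiers and repeat penalties.
; The dedicated server predicts on storm shrink, picks locations via EQS and spawns on storm hold, using a separate phase event for the first circle.
; It also manages pawn ability sets across respawns and broadcasts UI and sound cues.
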(event Events|EventOnGamePhaseUpdatedHandleFirstStormCircle
  (params (Event_GamePhase :type EAthenaGamePhase%20Enum))
  ; We are using another event to handle first storm anomalies, since the regular one is returning NONE and changing that may break other things
  (if (Variables|Default|GetHandledFirstSafeZone)
    _
    (elif (Math|Boolean|ANDBoolean (Networking|IsDedicatedServer) (CallFunction|IsSystemEnabled) (Default|IsSafeZoneSpawnActive))
      (switch Utilities|FlowControl|Switch|SwitchonEAthenaGamePhase Event_GamePhase
        (:Warmup
          (Development|PrintString "Checking for storm circle 1 anomalies")
          (Prediction|SetupNextPhaseRA)
          (Utilities|IsValid (Variables|Default|GetIncomingRA)
            (:Is%20Not%20Valid
              (Variables|Default|SetHandledFirstSafeZone true))))
        (:SafeZones
          (Variables|Default|SetHandledFirstSafeZone true)
          (CallFunction|EndActiveRiftAnomalies false)
          (Development|PrintString "Game Phase Calls: STORM HOLDING")
          (bind _incomingra (Variables|Default|GetIncomingRA)
            (:then
              (Spawning|TrySpawnRiftAnomaly (Variables|Default|GetSpawnLocation) _incomingra))
            (:else
              (CallFunction|UIClearActiveUI))))))))

(event Events|EventOnHandlePlaylistDataReady
  (params (GameState :type Fort%20Game%20State%20Athena%20Object%20Reference) (Playlist :type Fort%20Playlist%20Object%20Reference) (PlaylistContextTags :type Gameplay%20Tag%20Container%20Structure%20%28by%20ref%29))
  ;[0.3437 1 0.1827] Begin Play -> Setup listener for storm circle change
  (if (Math|Boolean|ANDBoolean (Networking|IsDedicatedServer) (CallFunction|IsSystemEnabled) (not (CallFunction|ShouldInitonGamePhaseStep)))
    (CallFunction|InitializeRiftAnomalies)))

(event Events|EventOnGamePhaseStepChanged
  (params (Event :type Fort%20Game%20Phase%20Step%20Updated%20Event%20Structure%20%28by%20ref%29))
  (bind _bisresumed (Default|IsSafeZoneSpawnActive))
  ;[0.3437 1 0.1827] Storm Circle Updated -> roll chance for new Rift Anomaly and run an EQS if the new Rift Anomaly needs one
  (if (and (Networking|IsDedicatedServer) (CallFunction|IsSystemEnabled))
    (switch Utilities|FlowControl|Switch|SwitchonEAthenaGamePhaseStep (Utilities|Struct|BreakFortGamePhaseStepUpdatedEvent Event)
      (:GetReady
        (if (CallFunction|ShouldInitonGamePhaseStep)
          (CallFunction|InitializeRiftAnomalies)))
      (:StormHolding
        (if _bisresumed
          (CallFunction|EndActiveRiftAnomalies false)
          (if (Utilities|IsValidSoftObjectReference (Variables|RiftAnomaly|GetCurrentRetryingAnomaly))
            (bind _returnvalue (CallFunction|ExecuteAnomalyCanSpawnRule (Variables|RiftAnomaly|GetCurrentRetryingAnomaly) (Variables|Default|GetIs%20RetryingActive)))
            (if _returnvalue
              (Spawning|TrySpawnRiftAnomaly (Variables|Default|GetSpawnLocation) (Utilities|ResolveSoftReference (Variables|RiftAnomaly|GetCurrentRetryingAnomaly)))
              (CallFunction|SetCurrentRetryingAnomaly 0))
            (else
              (bind _incomingra (Variables|Default|GetIncomingRA)
                (:then
                  (Spawning|TrySpawnRiftAnomaly (Variables|Default|GetSpawnLocation) _incomingra))
                (:else
                  (CallFunction|UIClearActiveUI)))))))
      (:StormShrinking
        (if _bisresumed
          (Prediction|SetupNextPhaseRA))))))

(event Events|EventOnSafeZoneUpdated
  (params (Event :type Fort%20Safe%20Zone%20Phase%20Updated%20Event%20Structure%20%28by%20ref%29))
  ;[0.3437 1 0.1827] Storm Circle Updated -> roll chance for new Rift Anomaly and run an EQS if the new Rift Anomaly needs one
  (Variables|Default|SetCurrentSafeZoneInfo (SafeZone|GetPhaseInfo (SafeZone|GetSafeZoneIndicator) -1)))

(event Custom|OnQueryFinished
  (params (QueryInstance :type EQS%20Query%20Instance%20Object%20Reference) (QueryStatus :type EEnvQueryStatus%20Enum :default "Processing"))
  ;[0.3437 1 0.1827] Spawn at a valid location once EQS finishes running
  (switch Utilities|FlowControl|Switch|SwitchonEEnvQueryStatus QueryStatus
    (:Success
      (bind (_resultlocations _returnvalue) (AI|EQS|GetQueryResultsasLocations QueryInstance))
      (Variables|Default|SetSpawnLocation (Math|Vector|SelectVector (Utilities|Array|Get(acopy) _resultlocations) (Utilities|Struct|BreakFortSafeZonePhaseInfo (Variables|Default|GetCurrentSafeZoneInfo)) _returnvalue))
      (Prediction|SetupAnticipation))
    (:Failed
      (Variables|Default|SetSpawnLocation (Utilities|Struct|BreakFortSafeZonePhaseInfo (Variables|Default|GetCurrentSafeZoneInfo)))))
  (CallFunction|AddAnomalyInfo (Utilities|String|ToString(Text) (Class|FortRiftAnomaliesData|GetRAName (Variables|Default|GetIncomingRA) :!class /Script/RiftAnomaliesGameplayRuntime.FortRiftAnomaliesData)) (Variables|Default|GetSpawnLocation)))

(event EventEndPlay
  (params (EndPlayReason :type EEndPlayReason%20Enum :default "Destroyed"))
  ;[0.3437 1 0.1827] End Play -> end listeners and all active MMFs
  (CallFunction|EndActiveRiftAnomalies true)
  (CallFunction|UIClearActiveUI)
  (Utilities|Time|ClearandInvalidateTimerbyHandle (Variables|Timers|GetSpawnPrediction_TimerHandle)))

(event Custom|UI_BroadcastIncomingRA
  (params (IncomingRIftAnomalyData :type Fort%20Rift%20Anomalies%20Data%20Object%20Reference))
  ;[0.8802 0.6674 1] Cosmetics -> handle adding and removing UI indicators
  (UI|RAIncomingClientUIActions IncomingRIftAnomalyData))

(event Custom|DebugForceSpawn_OnQueryFinished
  (params (QueryInstance :type EQS%20Query%20Instance%20Object%20Reference) (QueryStatus :type EEnvQueryStatus%20Enum :default "Processing"))
  ;[1 0 1 0.5] DEBUG SPAWN FLOW
  (Utilities|FlowControl|Sequence
    (:then_0
      (switch Utilities|FlowControl|Switch|SwitchonEEnvQueryStatus QueryStatus
        (:Success
          (bind (_resultlocations _returnvalue) (AI|EQS|GetQueryResultsasLocations QueryInstance))
          (if _returnvalue
            (Variables|Debug|SetDebugSpawnLocation (Utilities|Array|Get(acopy) _resultlocations))
            (else
              _)))
        (:Failed
          _))
      (bind _outactors (Actor|GetAllActorsOfClass "/Script/FortniteGame.FortPlayerPawnAthena"))
      (Variables|Debug|SetDebugSpawnLocation (Transformation|GetActorLocation (Utilities|Array|Get(acopy) _outactors))))
    (:then_1
      (Debug|DebugForceSpawnRA (Variables|Debug|GetDebugSpawnLocation) (Variables|Debug|GetDebugRAToSpawn)))))

(event Custom|UI_BroadcastRAActivated
  (params (ActiveAnomalyData :type Fort%20Rift%20Anomalies%20Data%20Object%20Reference))
  ;[0.8802 0.6674 1] Cosmetics -> handle adding and removing UI indicators
  (UI|RAActivatedClientUIActions ActiveAnomalyData))

(event Custom|UI_BroadcastRAEndedWithNotification
  (params (RiftAnomalyData :type Fort%20Rift%20Anomalies%20Data%20Object%20Reference))
  ;[0.8802 0.6674 1] Cosmetics -> handle adding and removing UI indicators
  (UI|RAEndedwithNotificationClientUIActions RiftAnomalyData))

(event Custom|EndAmbientRAEffects)

(event Custom|UI_ClearActiveUI
  ;[0.8802 0.6674 1] Cosmetics -> handle adding and removing UI indicators
  (UI|RAClearActiveClientUIActions))

(event Custom|Sound_BroadcastRAPlayTriggerSound
  (params (SoundToPlay :type Sound%20Base%20Object%20Reference))
  ;[0.8802 0.6674 1] Cosmetics -> handle adding and removing UI indicators
  (Utilities|IsValid SoundToPlay
    (:Is%20Valid
      (CallFunction|PlayRATriggerSounds SoundToPlay))))

(event Custom|Sound_BroadcastRAEndedWithNotification
  (params (Sound%20to%20Play :type Sound%20Base%20Object%20Reference))
  ;[0.8802 0.6674 1] Cosmetics -> handle adding and removing UI indicators
  (Utilities|IsValid Sound%20to%20Play
    (:Is%20Valid
      (CallFunction|PlayRAEndSound Sound%20to%20Play))))

(event EventOnPlayerPawnSpawned
  (params (OwningController :type Controller%20Object%20Reference))
  ;[1 1 1] On respawn, cleanup and then re-apply the effects
  (bind _asfort_pawn (Utilities|Casting|CastToFortPawn (Pawn|GetControlledPawn OwningController)))
  (CallFunction|RemoveRAEffects _asfort_pawn (Variables|Default|GetActiveRA))
  (CallFunction|ApplyRAEffects _asfort_pawn (Variables|Default|GetActiveRA)))

(event Events|EventOnCurrentRetryingAnomalyRep
  (bind _incomingra (Variables|Default|GetIncomingRA))
  (if (not (Networking|IsDedicatedServer))
    (Utilities|IsValid _incomingra
      (:Is%20Valid
        (CallFunction|LoadAssetsfromRADataAsset _incomingra)))))

(fn ApplyRAEffects
  (params (Pawn :type Fort%20Pawn%20Object%20Reference) (RAData :type Fort%20Rift%20Anomalies%20Data%20Object%20Reference))
  (bind _ballowed (CallFunction|CanApplyRAEffectsTo Pawn RAData))
  (if _ballowed
    (bind _output_get (LocalVariables|SetOwningController (Pawn|GetController (Variables|GetPawn))))
    (Utilities|IsValid _output_get
      (:Is%20Valid
        (switch Utilities|FlowControl|Switch|SwitchonERAEffectsRules (Class|FortRiftAnomaliesData|GetEffectApplicationRules (Variables|GetRAData))
          (:ApplyToAllPlayers
            (bind _returnvalue (Ability|EquipFortAbilitySet (Variables|GetPawn) (Class|FortRiftAnomaliesData|GetAbilitySet (Variables|GetRAData)) self))
            (Utilities|Map|Add (Variables|Default|GetAbilitySetHandles) (LocalVariables|GetOwningController) _returnvalue)
            (CallFunction|BindtoPlayerPawnSpawn (LocalVariables|GetOwningController))))))))

(fn B_RiftAnomaliesGameStateComponent_AutoGenFunc
  (bind _npclapsedtime (Variables|Default|GetNPCLapsedTime))
  (bind _returnvalue (+ _npclapsedtime 0.05))
  ;[1 1 1] MPC RiftAnomalyActive
  (bind _output_get (Variables|Default|SetNPCLapsedTime _returnvalue))
  (Rendering|Material|SetScalarParameterValue "/Game/Packages/Fortress_SharedMaterials/GlobalMaterialParameters/FortniteMaterialParameters.FortniteMaterialParameters" "RiftAnomalyActive" (Math|Float|MapRangeClamped _returnvalue 0.0 2.0 0.0 1.0) :!class /Script/Engine.KismetMaterialLibrary)
  (if (>= _output_get 2.0)
    (Utilities|Time|ClearandInvalidateTimerbyHandle (Variables|Default|GetNPCTimer))))

(fn CanApplyRAEffectsTo
  (params (Pawn :type Fort%20Pawn%20Object%20Reference%20%28by%20ref%29) (RAData :type Fort%20Rift%20Anomalies%20Data%20Object%20Reference%20%28by%20ref%29))
  (for _array_element (Variables|DataRegistry|GetPawnQueryTags)
    (if (not (GameplayTags|DoesContainerMatchTagQuery (GameplayTags|GetOwnedGameplayTags (Variables|GetPawn) :!class /Script/GameplayTags.BlueprintGameplayTagLibrary) (GameplayTags|MakeGameplayTagQuery _array_element)))
      (return false)))
  (return true))

(fn CheckAndApplySubsetOverrides
  (bind (_output_raname _output_ratag) (Utilities|Struct|BreakFortRAChance (Utilities|Array|Get(acopy) (LocalVariables|GetWeigthsChanges) (LocalVariables|GetSelectedWeightIndex))))
  (bind (_riftanomalysystemconfigdata_perstormcirclegeneralspawnchance_26_aaeceaa646d26e24c4340f8a66ffba3b _riftanomalysystemconfigdata_minsafezone_40_6abc48b2422b44b585ba67a5f74173a6 _riftanomalysystemconfigdata_maxsafezone_41_f34071834b0d86641a5860a838c2db3c _riftanomalysystemconfigdata_predictionwindow_37_4ad963b14b0f959ff7c44898407c2348 _riftanomalysystemconfigdata_cananomaliesrepeat_44_6ce45a574f0d808a542d9fb5e0603294 _riftanomalysystemconfigdata_rasystemenabled_46_fdecd3964c50a8aaab93158bbb40a992 _riftanomalysystemconfigdata_decreaseweightofanomalyinarow_50_39fec66249842b8db450b59845657a63 _riftanomalysystemconfigdata_maxsafezonespermatch_107_9a5f555f4dd71278a52e6b952b5e362a _riftanomalysystemconfigdata_overridestormweightbyphaserowbytag_60_9a5ba2b34503c08e68ce4f841220589b _riftanomalysystemconfigdata_chanceofspawnsafezonebasedanomalies_71_0941dfcb4c5215d20a9246a310499acb _riftanomalysystemconfigdata_anomalysubsetoverriding_70_82b0e46f433a27407a9f93a568955f28) (Utilities|Struct|BreakRiftAnomaliesAssetRegistrySystemData (Variables|GetRiftAnomalySystemConfigData_PerStormCircleGeneralSpawnChance_26_AAECEAA646D26E24C4340F8A66FFBA3B)))
  (bind _subsetoverridemap (LocalVariables|GetSubsetOverrideMap))
  (LocalVariables|SetMaxIndex _riftanomalysystemconfigdata_maxsafezonespermatch_107_9a5f555f4dd71278a52e6b952b5e362a)
  (bind _output_get (LocalVariables|SetSubsetOverrideMap _riftanomalysystemconfigdata_anomalysubsetoverriding_70_82b0e46f433a27407a9f93a568955f28))
  (if (> (Utilities|Map|Length _output_get) 0)
    (for _index (range (+ (LocalVariables|GetMaxIndex) 1))
      (LocalVariables|SetLoopIndex _index)
      (bind (_value _returnvalue) (Utilities|Map|Find _subsetoverridemap _index))
      (bind (_value_anomalysubsetweightsbysafezone _value_allowregularanomaliestospawn _value_chanceofstompsafezoneindexusingthissubset) (Utilities|Struct|BreakFortRAOverrideSafeZoneData _value))
      (if _returnvalue
        (if _value_allowregularanomaliestospawn
          ;[1 1 1] Check if subset can override given index
          (if (>= _value_chanceofstompsafezoneindexusingthissubset (Math|Random|RandomFloatinRange 0.0 1.0))
            _))
        (if (> (Utilities|Map|Length _value_anomalysubsetweightsbysafezone) 0)
          (bind (_keypin _valuepin) (Utilities|Map|ForEachLoop(Map) _value_anomalysubsetweightsbysafezone)
            (:then
              (Utilities|Array|Add (LocalVariables|GetWeigthsChanges) (Utilities|Struct|MakeFortRAChance (GameplayTags|GetTagName _keypin) _keypin _valuepin)))
            (:CompletedPin
              (CallFunction|RASelectAnomalyfromWeightedStructure (LocalVariables|GetLoopIndex) (LocalVariables|GetWeigthsChanges) (LocalVariables|GetSelectedWeightIndex))
              (bind (_anomaly _sucess) (DataRegistry|FindDataRegistrybyTag _output_ratag))
              (if _sucess
                (Utilities|Array|SetArrayElem (Variables|DataRegistry|GetPredefinedCircleAnomaliesDrops) (LocalVariables|GetLoopIndex) _anomaly true)
                (Utilities|Array|Clear (LocalVariables|GetWeigthsChanges))
                ;[1 1 1] Load anomaly on server
                (CallFunction|LoadAssetsfromRADataAsset (Utilities|Struct|BreakRiftAnomaliesAssetRegistryData _anomaly))))))))
    (else
      (return))))

(fn CheckAndApplyWeightModifiers
  (params (WeightToModify :type Integer%20%28by%20ref%29) (RATag :type Gameplay%20Tag%20Structure))
  (bind _ratag (Variables|GetRATag))
  (bind _raactivemodifiers (Variables|RAModifiers|GetRAActiveModifiers))
  (bind (_value _returnvalue) (Utilities|Map|Find _raactivemodifiers _ratag))
  (bind (_value_raweightmodifier _value_sumramodifier _value_blockthisanomalyfromspawn) (Utilities|Struct|BreakFortRAWeightModifier _value))
  (if (> WeightToModify 0)
    (if (Utilities|Map|Contains (Variables|RAModifiers|GetRAActiveModifiers) (Variables|GetRATag))
      (if _returnvalue
        (if _value_blockthisanomalyfromspawn
          ;[1 1 1] AnomalyBlocked
          (Variables|SetBy-RefVar (Variables|GetWeightToModify))
          (else
            (Variables|SetBy-RefVar (Variables|GetWeightToModify) (+ (Variables|GetWeightToModify) _value_raweightmodifier))
            (Development|PrintString (Utilities|String|Append (GameplayTags|GetDebugStringfromGameplayTag (Variables|GetRATag)) "  weight modified due to weight modifier" " new weight value: " (Utilities|String|ToString(Integer) (Variables|GetWeightToModify))))))))))

(fn CheckIfCanActivateStartMatchAnomalyUI
  (switch Utilities|FlowControl|Switch|SwitchonBP_Enum_RA_UI_STATE (Variables|ui|GetRiftAnomaliesUIState)
    (:NewEnumerator4
      _)
    (:NewEnumerator1
      _))
  (Utilities|IsValid (Variables|Default|GetMatchStartAnomalyData)
    (:Is%20Valid
      (Events|CallOnRiftAnomalyActivated (Variables|Default|GetMatchStartAnomalyData))
      (Variables|ui|SetRiftAnomaliesUIState "NewEnumerator5"))))

(fn CheckWeightSystemDataOverride
  (params (RATag :type Gameplay%20Tag%20Structure) (WeightToOverride :type Integer%20%28by%20ref%29) (StormCircleIndex :type Integer))
  (bind (_riftanomalysystemconfigdata_perstormcirclegeneralspawnchance_26_aaeceaa646d26e24c4340f8a66ffba3b _riftanomalysystemconfigdata_minsafezone_40_6abc48b2422b44b585ba67a5f74173a6 _riftanomalysystemconfigdata_maxsafezone_41_f34071834b0d86641a5860a838c2db3c _riftanomalysystemconfigdata_predictionwindow_37_4ad963b14b0f959ff7c44898407c2348 _riftanomalysystemconfigdata_cananomaliesrepeat_44_6ce45a574f0d808a542d9fb5e0603294 _riftanomalysystemconfigdata_rasystemenabled_46_fdecd3964c50a8aaab93158bbb40a992 _riftanomalysystemconfigdata_decreaseweightofanomalyinarow_50_39fec66249842b8db450b59845657a63 _riftanomalysystemconfigdata_maxsafezonespermatch_107_9a5f555f4dd71278a52e6b952b5e362a _riftanomalysystemconfigdata_overridestormweightbyphaserowbytag_60_9a5ba2b34503c08e68ce4f841220589b) (Utilities|Struct|BreakRiftAnomaliesAssetRegistrySystemData (Variables|GetRiftAnomalySystemConfigData_PerStormCircleGeneralSpawnChance_26_AAECEAA646D26E24C4340F8A66FFBA3B)))
  (bind (_value _returnvalue) (Utilities|Map|Find _riftanomalysystemconfigdata_overridestormweightbyphaserowbytag_60_9a5ba2b34503c08e68ce4f841220589b RATag))
  (bind _value_raweightpercirclemap (Utilities|Struct|BreakFortRAChanceMap _value))
  (bind _stormcircleindex (Variables|GetStormCircleIndex))
  (bind (_value_2 _returnvalue_2) (Utilities|Map|Find _value_raweightpercirclemap _stormcircleindex))
  (if _returnvalue
    (if _returnvalue_2
      (Variables|SetBy-RefVar (Variables|GetWeightToOverride) _value_2))))

(fn ClearFullMatchRAUI
  (Events|CallOnClearAnomaly))

(fn DebugForceSpawn_PerformEQS
  (bind _currentsafezoneinfo (Variables|Default|GetCurrentSafeZoneInfo))
  (bind (_center _radius) (Utilities|Struct|BreakFortSafeZonePhaseInfo _currentsafezoneinfo))
  (bind _returnvalue (AI|EQS|RunEQSQuery (Class|FortRiftAnomaliesData|GetConfiguredEQS (Variables|Debug|GetDebugRAToSpawn)) self))
  (AI|EQS|BindEventtoOnQueryFinished _returnvalue (EventDispatchers|CreateEvent :!event DebugForceSpawn_OnQueryFinished))
  (AI|EQS|SetNamedParam _returnvalue "GridSize" _radius)
  (AI|EQS|SetNamedParam _returnvalue "WorldLocationParamX" (.x _center))
  (AI|EQS|SetNamedParam _returnvalue "WorldLocationParamY" (.y _center))
  (AI|EQS|SetNamedParam _returnvalue "WorldLocationParamZ" (.z _center)))

(fn DebugForceSpawnRA
  (params (Location :type Vector) (Data :type Fort%20Rift%20Anomalies%20Data%20Object%20Reference%20%28by%20ref%29))
  (bind _returnvalue (Class|FortRiftAnomaliesData|GetRiftAnomalieActorClass Data))
  (bind (_riftanomalysystemconfigdata_perstormcirclegeneralspawnchance_26_aaeceaa646d26e24c4340f8a66ffba3b _riftanomalysystemconfigdata_minsafezone_40_6abc48b2422b44b585ba67a5f74173a6 _riftanomalysystemconfigdata_maxsafezone_41_f34071834b0d86641a5860a838c2db3c _riftanomalysystemconfigdata_predictionwindow_37_4ad963b14b0f959ff7c44898407c2348 _riftanomalysystemconfigdata_cananomaliesrepeat_44_6ce45a574f0d808a542d9fb5e0603294 _riftanomalysystemconfigdata_rasystemenabled_46_fdecd3964c50a8aaab93158bbb40a992 _riftanomalysystemconfigdata_decreaseweightofanomalyinarow_50_39fec66249842b8db450b59845657a63 _riftanomalysystemconfigdata_maxsafezonespermatch_107_9a5f555f4dd71278a52e6b952b5e362a _riftanomalysystemconfigdata_overridestormweightbyphaserowbytag_60_9a5ba2b34503c08e68ce4f841220589b _riftanomalysystemconfigdata_chanceofspawnsafezonebasedanomalies_71_0941dfcb4c5215d20a9246a310499acb _riftanomalysystemconfigdata_anomalysubsetoverriding_70_82b0e46f433a27407a9f93a568955f28 _riftanomalysystemconfigdata_usechancebyanomalysystemmode_84_c460390f49dae179dabf328c947469ec _riftanomalysystemconfigdata_weightsforquantityofanomaliesinamatch_97_7eecd97b46b44ef99cd0db914f6f2dcd _riftanomalysystemconfigdata_weightofananomalytobesortedonweightedsafezones_102_c579e04941f00bb8e72f949d6856433b _riftanomalysystemconfigdata_defaultweightforanomalyselection_106_86131a7644fbc538c8a7c7911131ce71 _riftanomalysystemconfigdata_defaultsafezoneweights_110_d4f62c4d4b770b30b08cf3acd19dd541 _riftanomalysystemconfigdata_prioritizeanomalywhiledecidingsafezones_114_9dbf0cf64512cfb0f4864b9ecdd77549 _riftanomalysystemconfigdata_anomaliesgroupsconfiguration_118_57b7639f457cde58dce1cfa3ccb50010 _riftanomalysystemconfigdata_uinotificationsenabled_121_4cd43fbe4608f5746aa5908accc4cdd6) (Utilities|Struct|BreakRiftAnomaliesAssetRegistrySystemData (Variables|GetRiftAnomalySystemConfigData_PerStormCircleGeneralSpawnChance_26_AAECEAA646D26E24C4340F8A66FFBA3B)))
  (Utilities|IsValid Data
    (:Is%20Valid
      (bind _anomalydataregistrydata (DataRegistry|FindRegistrybyDataAsset Data))
      (bind (_anomalydataregistrydata_anomaliesdataassets_20_7e8609f440fac2369e8ac7902781fa79 _anomalydataregistrydata_stormweights_26_be7c04904bf803d046fa5282e362cc01 _anomalydataregistrydata_canoccurmultipletimes_16_53ee0f1f44ebe0d30c76e08a31e8c644 _anomalydataregistrydata_effectradius_24_aaeceaa646d26e24c4340f8a66ffba3b _anomalydataregistrydata_isenabled_30_aac59cb249d32975a549d19e98331703 _anomalydataregistrydata_playlisttagquery_38_760275694a7508c4a9500d8393a0ca1e _anomalydataregistrydata_stormweightbyphaserow_43_f8665c0a434e4870bac26abb909a92d3 _anomalydataregistrydata_ratag_47_3bb6e306408e53230fe68fac1bcd5bb3 _anomalydataregistrydata_ramodfiermap_59_87219a6747987167eb1cc3945bd39cf7 _anomalydataregistrydata_lastanomaly_62_78d3c0494913613382169fb5d22a26d2 _anomalydataregistrydata_forceshowui_64_3859b8cf4298278ed0f3329e21080f62) (Utilities|Struct|BreakRiftAnomaliesAssetRegistryData _anomalydataregistrydata))
      (LocalVariables|SetForceShowUI _anomalydataregistrydata_forceshowui_64_3859b8cf4298278ed0f3329e21080f62)
      (if (Utilities|IsValidClass _returnvalue)
        (bind _returnvalue_2 (Game|SpawnActorfromClass _returnvalue (Math|Transform|MakeTransform Location) "AlwaysSpawn" "MultiplyWithRoot" 0 (Variables|GetData) 0))
        (LocalVariables|SetSpawnedRAActorBase _returnvalue_2)
        (Debugging|PrintVector (Transformation|GetActorLocation (LocalVariables|GetSpawnedRAActorBase)) (Utilities|String|ToString(Text) (Utilities|Text|FormatText "NSLOCTEXT(\"[4B9D47DCB44E2CC832FF8DB0E2DA666C]\", \"0F70BB984777F95885F86B84D92032D0\", \"RA \\\"{a}\\\" successfully spawned at {b}\")" (Variables|/Script/RiftAnomaliesGameplayRuntime.FortRiftAnomaliesData|GetRAName (Variables|GetData)) (Utilities|String|ToString(Vector) (Transformation|GetActorLocation (LocalVariables|GetSpawnedRAActorBase))))) 12.0 "(R=0.076514,G=1.000000,B=0.001511,A=1.000000)"
          (:Then
            _)))
      (Utilities|Array|AddUnique (Variables|Default|GetAllActiveRAInfo) (Utilities|Struct|MakeBRAActiveInfoStruct (Variables|GetData) (LocalVariables|GetSpawnedRAActorBase)))
      (Variables|Default|SetActiveRA (Variables|GetData))
      (switch Utilities|FlowControl|Switch|SwitchonERAEffectsRules (Class|FortRiftAnomaliesData|GetEffectApplicationRules (Variables|GetData))
        (:DontApplyEffects
          _)
        (:ApplyToAllPlayers
          (bind _outfortplayerpawns (Fortnite|GetFortPlayerPawns))
          (for _array_element _outfortplayerpawns
            (Utilities|IsValid _array_element
              (:Is%20Valid
                (CallFunction|ApplyRAEffects _array_element (Variables|GetData)))))
          (CallFunction|ListenforPawnSpawnforDeadPlayers))
        (:ApplyViaOverlap
          _))
      (if (or (LocalVariables|GetForceShowUI) _riftanomalysystemconfigdata_uinotificationsenabled_121_4cd43fbe4608f5746aa5908accc4cdd6)
        (CallFunction|UIBroadcastRAActivated (Variables|Default|GetActiveRA)))
      (CallFunction|SoundBroadcastRAPlayTriggerSound (Class|FortRiftAnomaliesData|GetRiftAnomalyTriggerSound (Variables|Default|GetActiveRA)))
      (Variables|Debug|SetDebugRAToSpawn 0)
      (Spawning|SpawnAnomalyFX))))

(fn DoAnomalyTypeActions
  (params (AnomalyToCheck :type Fort%20Rift%20Anomalies%20Data%20Object%20Reference))
  ; This supports just one start match anomaly and is created in a hurry to support storm infill as a rift anomaly
  (if (Utilities|Enum|Equal(Enum) (Class|FortRiftAnomaliesData|GetRABehaviorType AnomalyToCheck) "MatchStartOrientedRA")
    (Utilities|IsValid (Variables|Default|GetMatchStartAnomalyData)
      (:Is%20Valid
        (Fortnite|FortShippingLog self "LogFortRAManager: There's already a match start anomaly set" true))
      (:Is%20Not%20Valid
        (Variables|Default|SetMatchStartAnomalyData (Variables|Default|GetMatchStartAnomalyData))))))

(fn EndActiveRiftAnomalies
  (params (ForceEndFullMatch :type Boolean))
  (bind _allactiverainfo (Variables|Default|GetAllActiveRAInfo))
  (if (Utilities|Array|IsNotEmpty _allactiverainfo)
    (CallFunction|HandleQuestEvent)
    (for _array_element _allactiverainfo
      (bind (_array_element_activeradata_6_893472d943ba8730134436b6a7e04009 _array_element_activeraactor_5_2547e6b1486e354d5ac3e7aa00ec8c37) (Utilities|Struct|BreakBRAActiveInfoStruct _array_element))
      (LocalVariables|SetCurrentIteratedRAData _array_element_activeradata_6_893472d943ba8730134436b6a7e04009)
      (Utilities|IsValid (LocalVariables|GetCurrentIteratedRAData)
        (:Is%20Valid
          (if (or (not (CallFunction|IsFullMatchRiftAnomaly (LocalVariables|GetCurrentIteratedRAData))) (Variables|GetForceEndFullMatch))
            (Utilities|IsValid _array_element_activeraactor_5_2547e6b1486e354d5ac3e7aa00ec8c37
              (:Is%20Valid
                (Class|RiftAnomaliesActorBase|EndRiftAnomaly _array_element_activeraactor_5_2547e6b1486e354d5ac3e7aa00ec8c37))
              (:Is%20Not%20Valid
                _))
            (Utilities|IsValid (LocalVariables|GetCurrentIteratedRAData)
              (:Is%20Valid
                (switch Utilities|FlowControl|Switch|SwitchonERAEffectsRules (Class|FortRiftAnomaliesData|GetEffectApplicationRules (LocalVariables|GetCurrentIteratedRAData))
                  (:ApplyToAllPlayers
                    (bind _outfortplayerpawns (Fortnite|GetFortPlayerPawns))
                    (for _array_element_2 _outfortplayerpawns
                      (CallFunction|RemoveRAEffects _array_element_2 (LocalVariables|GetCurrentIteratedRAData))
                      ;[1 1 1] MPC RiftAnomalyActive
                      (Rendering|Material|SetScalarParameterValue "/Game/Packages/Fortress_SharedMaterials/GlobalMaterialParameters/FortniteMaterialParameters.FortniteMaterialParameters" "RiftAnomalyActive" :!class /Script/Engine.KismetMaterialLibrary)))))
              (:Is%20Not%20Valid
                (Development|PrintString "RAData not valid")))))))
    (else
      _))
  (Utilities|Array|Clear (Variables|Default|GetAllActiveRAInfo))
  (Variables|Default|SetActiveRA 0)
  (CallFunction|UnbindSpawnPawnListeners))

(fn FindDataRegistryByTag
  (params (AnomalyTag :type Gameplay%20Tag%20Structure))
  (bind _outidlist (DataRegistry|GetPossibleDataRegistryIdList "AnomaliesToLoad"))
  (for _array_element _outidlist
    (bind _outitem (DataRegistry|FindDataRegistryItem _array_element)
      (:Found
        (bind _output_get (LocalVariables|SetCurrentIteratedAnomaly _outitem))
        (bind (_output_get_anomaliesdataassets_20_7e8609f440fac2369e8ac7902781fa79 _output_get_stormweights_26_be7c04904bf803d046fa5282e362cc01 _output_get_canoccurmultipletimes_16_53ee0f1f44ebe0d30c76e08a31e8c644 _output_get_effectradius_24_aaeceaa646d26e24c4340f8a66ffba3b _output_get_isenabled_30_aac59cb249d32975a549d19e98331703 _output_get_playlisttagquery_38_760275694a7508c4a9500d8393a0ca1e _output_get_stormweightbyphaserow_43_f8665c0a434e4870bac26abb909a92d3 _output_get_ratag_47_3bb6e306408e53230fe68fac1bcd5bb3) (Utilities|Struct|BreakRiftAnomaliesAssetRegistryData _output_get))
        (if (== _output_get_ratag_47_3bb6e306408e53230fe68fac1bcd5bb3 (Variables|GetAnomalyTag))
          (return (LocalVariables|GetCurrentIteratedAnomaly) true)))))
  (return 0 false))

(fn FindRegistryByDataAsset
  (params (RiftAnomalyDataToSearch :type Fort%20Rift%20Anomalies%20Data%20Object%20Reference))
  (for _array_element (Variables|DataRegistry|GetPredefinedCircleAnomaliesDrops)
    (bind _output_get (LocalVariables|SetIteratedRegistryData _array_element))
    (Utilities|IsValid (Utilities|Struct|BreakRiftAnomaliesAssetRegistryData _output_get)
      (:Is%20Valid
        (if (== (Utilities|Struct|BreakRiftAnomaliesAssetRegistryData _output_get) (Variables|GetRiftAnomalyDataToSearch))
          (return (LocalVariables|GetIteratedRegistryData) 0)))))
  (return 0 0))

(fn ForceEndActiveRiftAnomalies
  (params (bWithEndedHUDNotification :type Boolean) (bShouldClearActiveUI :type Boolean :default true) (bBlockRAsForRestOfMatch :type Boolean))
  (bind _activera (Variables|Default|GetActiveRA))
  (Utilities|SwitchHasAuthority
    (:Authority
      (Utilities|FlowControl|Sequence
        (:then_0
          (Variables|GetbWithEndedHUDNotification
            (:then
              (Utilities|IsValid _activera
                (:Is%20Valid
                  (CallFunction|UIBroadcastRAEndedwithNotification _activera)
                  (CallFunction|SoundBroadcastRAEndedwithNotification (Class|FortRiftAnomaliesData|GetRiftAnomalyEndSound _activera)))))
            (:else)))
        (:then_1
          (CallFunction|EndActiveRiftAnomalies (Variables|GetbBlockRAsForRestOfMatch)))
        (:then_2
          (Variables|GetbShouldClearActiveUI
            (:then
              (CallFunction|UIClearActiveUI))
            (:else)))))))

(fn GetAnomalyAccordingToWeight
  (params (SafeZoneIndex :type Integer) (LastSortedAnomaly :type Name))
  (bind (_perstormcirclegeneralspawnchance_26_aaeceaa646d26e24c4340f8a66ffba3b _minsafezone_40_6abc48b2422b44b585ba67a5f74173a6 _maxsafezone_41_f34071834b0d86641a5860a838c2db3c _predictionwindow_37_4ad963b14b0f959ff7c44898407c2348 _cananomaliesrepeat_44_6ce45a574f0d808a542d9fb5e0603294 _rasystemenabled_46_fdecd3964c50a8aaab93158bbb40a992 _decreaseweightofanomalyinarow_50_39fec66249842b8db450b59845657a63) (Utilities|Struct|BreakRiftAnomaliesAssetRegistrySystemData (Variables|DataRegistry|GetRiftAnomalySystemConfigData)))
  (bind _sortedindex (LocalVariables|GetSortedIndex))
  (for _array_element (Variables|DataRegistry|GetPossibleAnomaliesToSpawn)
    (bind (_array_element_anomaliedata_2_f38e3dd0464ff4283aff31b4e94803e4 _array_element_rowname_5_4e12e4c6414e627e8ec7ce9fd91aad35) (Utilities|Struct|BreakBRADataRegistryWithRowNameStruct _array_element))
    (LocalVariables|SetScalableFloatForWeights (.StormWeightByPhaseRow_43_F8665C0A434E4870BAC26ABB909A92D3 (Utilities|Struct|BreakRiftAnomaliesAssetRegistryData _array_element_anomaliedata_2_f38e3dd0464ff4283aff31b4e94803e4)))
    (LocalVariables|SetCurrentAnomalyRowName _array_element_rowname_5_4e12e4c6414e627e8ec7ce9fd91aad35)
    (bind _output_get (LocalVariables|SetCurrentAnomalySafeZoneWeight (Math|Integer|SelectInt (Math|Float|Truncate (ScalableFloat|GetValueatLevel (LocalVariables|GetScalableFloatForWeights) (Math|Conversions|ToFloat(Integer) (Variables|GetSafeZoneIndex)))) (Utilities|Array|Get(acopy) (.StormWeights_26_BE7C04904BF803D046FA5282E362CC01 (Utilities|Struct|BreakRiftAnomaliesAssetRegistryData _array_element_anomaliedata_2_f38e3dd0464ff4283aff31b4e94803e4)) (Variables|GetSafeZoneIndex)) (ScalableFloat|IsSettingCurveValid (LocalVariables|GetScalableFloatForWeights)))))
    (CallFunction|CheckWeightSystemDataOverride (.RATag_47_3BB6E306408E53230FE68FAC1BCD5BB3 (Utilities|Struct|BreakRiftAnomaliesAssetRegistryData _array_element_anomaliedata_2_f38e3dd0464ff4283aff31b4e94803e4)) _output_get (Variables|GetSafeZoneIndex))
    (RAModifiers|CheckandApplyWeightModifiers _output_get (.RATag_47_3BB6E306408E53230FE68FAC1BCD5BB3 (Utilities|Struct|BreakRiftAnomaliesAssetRegistryData _array_element_anomaliedata_2_f38e3dd0464ff4283aff31b4e94803e4)))
    (if (> (LocalVariables|GetCurrentAnomalySafeZoneWeight) 0)
      (if (== (LocalVariables|GetCurrentAnomalyRowName) (Variables|GetLastSortedAnomaly))
        ;[1 1 1] DecreasingChanceOfAnomaliesInARow
        (Utilities|Array|Add (LocalVariables|GetSafeZoneWeights) (Utilities|Struct|MakeFortRAChance (LocalVariables|GetCurrentAnomalyRowName) "(TagName=\"\")" (- (LocalVariables|GetCurrentAnomalySafeZoneWeight) _decreaseweightofanomalyinarow_50_39fec66249842b8db450b59845657a63)))
        (else
          (Utilities|Array|Add (LocalVariables|GetSafeZoneWeights) (Utilities|Struct|MakeFortRAChance (LocalVariables|GetCurrentAnomalyRowName) "(TagName=\"\")" (LocalVariables|GetCurrentAnomalySafeZoneWeight)))))
      (else
        (Development|PrintString (Utilities|String|Append "Weight of anomaly: " (Utilities|String|ToString(Name) (LocalVariables|GetCurrentAnomalyRowName)) " is 0 at index " (Utilities|String|ToString(Integer) (Variables|GetSafeZoneIndex)))))))
  (if (Utilities|Array|IsNotEmpty (LocalVariables|GetSafeZoneWeights))
    (CallFunction|RASelectAnomalyfromWeightedStructure (Variables|GetSafeZoneIndex) (LocalVariables|GetSafeZoneWeights) _sortedindex)
    (return (Utilities|Array|Get(acopy) (Variables|DataRegistry|GetPossibleAnomaliesToSpawn) _sortedindex) (LocalVariables|GetSortedIndex))))

(fn GetItemsFromRegistries
  (bind _currentiteratedanomaly (LocalVariables|GetCurrentIteratedAnomaly))
  (bind _currentiteratedanomaly_anomaliesdataassets_20_7e8609f440fac2369e8ac7902781fa79 (Utilities|Struct|BreakRiftAnomaliesAssetRegistryData (Variables|GetCurrentIteratedAnomaly_AnomaliesDataAssets_20_7E8609F440FAC2369E8AC7902781FA79)))
  (bind _outidlist (DataRegistry|GetPossibleDataRegistryIdList "AnomaliesToLoad"))
  (for _array_element _outidlist
    (bind _outitem (DataRegistry|FindDataRegistryItem _array_element)
      (:Found
        (bind _output_get (LocalVariables|SetCurrentIteratedAnomaly _outitem))
        (Utilities|Array|Add (Variables|DataRegistry|GetAllAnomaliesData) (Utilities|Struct|MakeBRADataRegistrywithRowNameStruct _outitem (.ItemName (Utilities|Struct|BreakDataRegistryId _array_element))))
        (if (CallFunction|ShouldAnomalyStayinthePool _output_get)
          (if (CallFunction|IsFullMatchRiftAnomaly _currentiteratedanomaly_anomaliesdataassets_20_7e8609f440fac2369e8ac7902781fa79)
            (Utilities|Array|Add (Variables|DataRegistry|GetFullMatchAnomalies) (Utilities|Struct|MakeBRADataRegistrywithRowNameStruct _currentiteratedanomaly (.ItemName (Utilities|Struct|BreakDataRegistryId _array_element))))
            (bind _anomaliedata_2_f38e3dd0464ff4283aff31b4e94803e4 (Utilities|Struct|BreakBRADataRegistrywithRowNameStruct (Utilities|Struct|MakeBRADataRegistrywithRowNameStruct _currentiteratedanomaly (.ItemName (Utilities|Struct|BreakDataRegistryId _array_element)))))
            (bind _anomaliedata_2_f38e3dd0464ff4283aff31b4e94803e4_anomaliesdataassets_20_7e8609f440fac2369e8ac7902781fa79 (Utilities|Struct|BreakRiftAnomaliesAssetRegistryData _anomaliedata_2_f38e3dd0464ff4283aff31b4e94803e4))
            (Utilities|IsValid _anomaliedata_2_f38e3dd0464ff4283aff31b4e94803e4_anomaliesdataassets_20_7e8609f440fac2369e8ac7902781fa79
              (:Is%20Valid
                ; Deactivated for now!
                (switch Utilities|FlowControl|Switch|SwitchonERABehavior (Class|FortRiftAnomaliesData|GetRABehaviorType _anomaliedata_2_f38e3dd0464ff4283aff31b4e94803e4_anomaliesdataassets_20_7e8609f440fac2369e8ac7902781fa79))))
            (else
              (Utilities|Array|Add (Variables|DataRegistry|GetPossibleAnomaliesToSpawn) (Utilities|Struct|MakeBRADataRegistrywithRowNameStruct (LocalVariables|GetCurrentIteratedAnomaly) (.ItemName (Utilities|Struct|BreakDataRegistryId _array_element))))))
          (else
            (Development|PrintString "Anomaly Removed From The pool "))))))
  (Utilities|FlowControl|Sequence
    (:then_0
      (bind _outidlist_2 (DataRegistry|GetPossibleDataRegistryIdList "AnomaliesSystemConfig"))
      (for _array_element_2 _outidlist_2
        (bind _outitem_2 (DataRegistry|FindDataRegistryItem _array_element_2)
          (:Found
            (bind _output_get_2 (Variables|DataRegistry|SetRiftAnomalySystemConfigData _outitem_2))
            (bind (_output_get_perstormcirclegeneralspawnchance_26_aaeceaa646d26e24c4340f8a66ffba3b _output_get_minsafezone_40_6abc48b2422b44b585ba67a5f74173a6 _output_get_maxsafezone_41_f34071834b0d86641a5860a838c2db3c _output_get_predictionwindow_37_4ad963b14b0f959ff7c44898407c2348 _output_get_cananomaliesrepeat_44_6ce45a574f0d808a542d9fb5e0603294 _output_get_rasystemenabled_46_fdecd3964c50a8aaab93158bbb40a992 _output_get_decreaseweightofanomalyinarow_50_39fec66249842b8db450b59845657a63 _output_get_maxsafezonespermatch_107_9a5f555f4dd71278a52e6b952b5e362a _output_get_overridestormweightbyphaserowbytag_60_9a5ba2b34503c08e68ce4f841220589b _output_get_chanceofspawnsafezonebasedanomalies_71_0941dfcb4c5215d20a9246a310499acb _output_get_anomalysubsetoverriding_70_82b0e46f433a27407a9f93a568955f28 _output_get_usechancebyanomalysystemmode_84_c460390f49dae179dabf328c947469ec _output_get_weightsforquantityofanomaliesinamatch_97_7eecd97b46b44ef99cd0db914f6f2dcd _output_get_weightofananomalytobesortedonweightedsafezones_102_c579e04941f00bb8e72f949d6856433b _output_get_defaultweightforanomalyselection_106_86131a7644fbc538c8a7c7911131ce71 _output_get_defaultsafezoneweights_110_d4f62c4d4b770b30b08cf3acd19dd541) (Utilities|Struct|BreakRiftAnomaliesAssetRegistrySystemData _output_get_2))
            (Development|PrintString "Finished reading system anomaly data registry")
            (Utilities|FlowControl|Sequence
              (:then_0
                (Variables|Default|SetRift%20Anomaly%20System%20Config%20Data%20Default%20Safe%20Zone%20Weights _output_get_defaultsafezoneweights_110_d4f62c4d4b770b30b08cf3acd19dd541))
              (:then_1
                (Variables|Default|SetMax%20Safe%20Zones%20Per%20Match _output_get_maxsafezonespermatch_107_9a5f555f4dd71278a52e6b952b5e362a))))))
      (if (Utilities|Array|IsEmpty _outidlist_2)
        (Development|PrintString "System data is missing")))
    (:then_1
      (bind _outidlist_3 (DataRegistry|GetPossibleDataRegistryIdList "AnomaliesPawnQueryConfig"))
      (for _array_element_3 _outidlist_3
        (bind _outitem_3 (DataRegistry|FindDataRegistryItem _array_element_3)
          (:Found
            (Utilities|Array|AddUnique (Variables|DataRegistry|GetPawnQueryTags) (Utilities|Struct|BreakTagQueryTableRow _outitem_3)))))
      (return))))

(fn GetRandomizedRA
  (params (ShouldCheckOccurMultipleTimes :type Boolean) (ShouldUseWeight :type Boolean))
  (bind (_outitem _outindex) (Utilities|Array|RandomArrayItem (Variables|DataRegistry|GetPossibleAnomaliesToSpawn)))
  (bind (_riftanomalysystemconfigdata_perstormcirclegeneralspawnchance_26_aaeceaa646d26e24c4340f8a66ffba3b _riftanomalysystemconfigdata_minsafezone_40_6abc48b2422b44b585ba67a5f74173a6 _riftanomalysystemconfigdata_maxsafezone_41_f34071834b0d86641a5860a838c2db3c _riftanomalysystemconfigdata_predictionwindow_37_4ad963b14b0f959ff7c44898407c2348 _riftanomalysystemconfigdata_cananomaliesrepeat_44_6ce45a574f0d808a542d9fb5e0603294) (Utilities|Struct|BreakRiftAnomaliesAssetRegistrySystemData (Variables|GetRiftAnomalySystemConfigData_PerStormCircleGeneralSpawnChance_26_AAECEAA646D26E24C4340F8A66FFBA3B)))
  (if (> (Utilities|Array|Length (Variables|DataRegistry|GetPossibleAnomaliesToSpawn)) 0)
    (bind _output_get (LocalVariables|SetSelectedData))
    (bind (_output_get_anomaliesdataassets_20_7e8609f440fac2369e8ac7902781fa79 _output_get_stormweights_26_be7c04904bf803d046fa5282e362cc01 _output_get_canoccurmultipletimes_16_53ee0f1f44ebe0d30c76e08a31e8c644) (Utilities|Struct|BreakRiftAnomaliesAssetRegistryData _output_get))
    (if _riftanomalysystemconfigdata_cananomaliesrepeat_44_6ce45a574f0d808a542d9fb5e0603294
      (if _output_get_canoccurmultipletimes_16_53ee0f1f44ebe0d30c76e08a31e8c644
        _
        (elif (Variables|GetShouldCheckOccurMultipleTimes)
          (Utilities|Array|RemoveIndex (Variables|DataRegistry|GetPossibleAnomaliesToSpawn) _outindex)))
      (Utilities|IsValid _output_get_anomaliesdataassets_20_7e8609f440fac2369e8ac7902781fa79
        (:Is%20Valid
          (return (LocalVariables|GetSelectedData) true))
        (:Is%20Not%20Valid
          (return 0 false))))))

(fn GetRandomizedRAConsideringWeights
  (params (ShouldCheckOccurMultipleTimes :type Boolean) (Index :type Integer) (LastSortedAnomaly :type Name))
  (bind (_riftanomalysystemconfigdata_perstormcirclegeneralspawnchance_26_aaeceaa646d26e24c4340f8a66ffba3b _riftanomalysystemconfigdata_minsafezone_40_6abc48b2422b44b585ba67a5f74173a6 _riftanomalysystemconfigdata_maxsafezone_41_f34071834b0d86641a5860a838c2db3c _riftanomalysystemconfigdata_predictionwindow_37_4ad963b14b0f959ff7c44898407c2348 _riftanomalysystemconfigdata_cananomaliesrepeat_44_6ce45a574f0d808a542d9fb5e0603294) (Utilities|Struct|BreakRiftAnomaliesAssetRegistrySystemData (Variables|GetRiftAnomalySystemConfigData_PerStormCircleGeneralSpawnChance_26_AAECEAA646D26E24C4340F8A66FFBA3B)))
  (bind (_riftanomalydata _anomalyindex) (CallFunction|GetAnomalyAccordingtoWeight Index LastSortedAnomaly))
  (bind _output_get (LocalVariables|SetSelectedData _riftanomalydata))
  (bind _output_get_anomaliedata_2_f38e3dd0464ff4283aff31b4e94803e4 (Utilities|Struct|BreakBRADataRegistryWithRowNameStruct _output_get))
  (LocalVariables|SetAnomalyRiftToCheck (Utilities|Struct|BreakRiftAnomaliesAssetRegistryData _output_get_anomaliedata_2_f38e3dd0464ff4283aff31b4e94803e4))
  (if _riftanomalysystemconfigdata_cananomaliesrepeat_44_6ce45a574f0d808a542d9fb5e0603294
    (if (.CanOccurMultipleTimes_16_53EE0F1F44EBE0D30C76E08A31E8C644 (Utilities|Struct|BreakRiftAnomaliesAssetRegistryData _output_get_anomaliedata_2_f38e3dd0464ff4283aff31b4e94803e4))
      _
      (elif (Variables|GetShouldCheckOccurMultipleTimes)
        (Utilities|Array|RemoveIndex (Variables|DataRegistry|GetPossibleAnomaliesToSpawn) _anomalyindex)))
    (LocalVariables|GetAnomalyRiftToCheck
      (:then
        (return (LocalVariables|GetSelectedData) true))
      (:else
        (return 0 false)))))

(fn HandleAddingNewRAModifiers
  (params (NewRAListOfModifiersToAdd :type Map%20of%20Gameplay%20Tag%20Structures%20to%20Fort%20RAWeight%20Modifier%20Structures))
  (bind (_keypin _valuepin) (Utilities|Map|ForEachLoop(Map) NewRAListOfModifiersToAdd))
  (LocalVariables|SetCurrentIteratedKey _keypin)
  (LocalVariables|SetCurrentIteratedModifier (Utilities|Struct|BreakFortRAWeightModifier _valuepin))
  (if (.BlockThisAnomalyFromSpawn (Utilities|Struct|BreakFortRAWeightModifier _valuepin))
    (CallFunction|RemoveAnomalyfromPossibleList "None" (LocalVariables|GetCurrentIteratedKey))
    (elif (and (.SumRAModifier (Utilities|Struct|BreakFortRAWeightModifier _valuepin)) (Utilities|Map|Contains (Variables|RAModifiers|GetRAActiveModifiers) (LocalVariables|GetCurrentIteratedKey)))
      ;[1 1 1] Increasing Modifier
      (bind _output_get (LocalVariables|SetCurrentMapValue (Utilities|Map|Find (Variables|RAModifiers|GetRAActiveModifiers) (LocalVariables|GetCurrentIteratedKey))))
      (bind (_output_get_raweightmodifier _output_get_sumramodifier _output_get_blockthisanomalyfromspawn) (Utilities|Struct|BreakFortRAWeightModifier _output_get))
      (bind _output_get_2 (Variables|SetCurrentMapValue_RAWeightModifier (Utilities|Struct|MakeFortRAWeightModifier (+ _output_get_raweightmodifier (LocalVariables|GetCurrentIteratedModifier)) false _output_get_blockthisanomalyfromspawn)))
      (Utilities|Map|Add (Variables|RAModifiers|GetRAActiveModifiers) (LocalVariables|GetCurrentIteratedKey) _output_get_2)
      (else
        ;[1 1 1] If the key already existis it will override the value
        (Utilities|Map|Add (Variables|RAModifiers|GetRAActiveModifiers) (LocalVariables|GetCurrentIteratedKey) _valuepin)))))

(fn HandleQuestEvent
  (bind _outfortplayerpawns (Fortnite|GetFortPlayerPawns))
  (for _array_element _outfortplayerpawns
    (if (not (Pawn|IsBotControlled _array_element))
      (Ability|SendGameplayEventtoActor _array_element (Variables|Default|GetEvent_RACompleted)))))

(fn InitializeRiftAnomalies
  (DataRegistry|GetItemsfromRegistries)
  (if (CallFunction|IsUsingChancebyAnomalyMode)
    (CallFunction|SystemonStormCircleWeightedMode)
    (else
      (CallFunction|SystemonWeightedAnomalySelectionMode)))
  (Determinism|CheckandApplySubsetOverrides))

(fn IsFullMatchRiftAnomaly
  (params (RiftAnomalyDataAsset :type Fort%20Rift%20Anomalies%20Data%20Object%20Reference))
  (Utilities|IsValid RiftAnomalyDataAsset
    (:Is%20Valid
      (switch Utilities|FlowControl|Switch|SwitchonERAType (Class|FortRiftAnomaliesData|GetRiftAnomalyType RiftAnomalyDataAsset)
        (:SafeZoneRA
          (return false))
        (:FullMatchRA
          (return true))))
    (:Is%20Not%20Valid
      (return false))))

(fn IsGivenStormCircleHaveAnomalies
  (params (Index :type Integer))
  (bind _riftanomalysystemconfigdata_perstormcirclegeneralspawnchance_26_aaeceaa646d26e24c4340f8a66ffba3b (Utilities|Struct|BreakRiftAnomaliesAssetRegistrySystemData (Variables|GetRiftAnomalySystemConfigData_PerStormCircleGeneralSpawnChance_26_AAECEAA646D26E24C4340F8A66FFBA3B)))
  (bind (_riftanomalysystemconfigdata_perstormcirclegeneralspawnchance_26_aaeceaa646d26e24c4340f8a66ffba3b_2 _riftanomalysystemconfigdata_minsafezone_40_6abc48b2422b44b585ba67a5f74173a6 _riftanomalysystemconfigdata_maxsafezone_41_f34071834b0d86641a5860a838c2db3c) (Utilities|Struct|BreakRiftAnomaliesAssetRegistrySystemData (Variables|GetRiftAnomalySystemConfigData_PerStormCircleGeneralSpawnChance_26_AAECEAA646D26E24C4340F8A66FFBA3B)))
  (if (Utilities|Array|IsValidIndex _riftanomalysystemconfigdata_perstormcirclegeneralspawnchance_26_aaeceaa646d26e24c4340f8a66ffba3b Index)
    (if (Math|Integer|InRange(Integer) Index _riftanomalysystemconfigdata_minsafezone_40_6abc48b2422b44b585ba67a5f74173a6 _riftanomalysystemconfigdata_maxsafezone_41_f34071834b0d86641a5860a838c2db3c)
      (bind _output_get (LocalVariables|SetRandomChance (Math|Random|RandomFloatinRange 0.0 1.0)))
      (if (>= (Utilities|Array|Get(aref) _riftanomalysystemconfigdata_perstormcirclegeneralspawnchance_26_aaeceaa646d26e24c4340f8a66ffba3b_2 (Variables|GetIndex)) _output_get)
        (return true)
        (else
          (return false)))
      (else
        (return false)))
    (else
      (return false))))

(fn IsGoingToSpawnAnySafeZoneBasedRAs
  (bind (_riftanomalysystemconfigdata_perstormcirclegeneralspawnchance_26_aaeceaa646d26e24c4340f8a66ffba3b _riftanomalysystemconfigdata_minsafezone_40_6abc48b2422b44b585ba67a5f74173a6 _riftanomalysystemconfigdata_maxsafezone_41_f34071834b0d86641a5860a838c2db3c _riftanomalysystemconfigdata_predictionwindow_37_4ad963b14b0f959ff7c44898407c2348 _riftanomalysystemconfigdata_cananomaliesrepeat_44_6ce45a574f0d808a542d9fb5e0603294 _riftanomalysystemconfigdata_rasystemenabled_46_fdecd3964c50a8aaab93158bbb40a992 _riftanomalysystemconfigdata_decreaseweightofanomalyinarow_50_39fec66249842b8db450b59845657a63 _riftanomalysystemconfigdata_maxsafezonespermatch_107_9a5f555f4dd71278a52e6b952b5e362a _riftanomalysystemconfigdata_overridestormweightbyphaserowbytag_60_9a5ba2b34503c08e68ce4f841220589b _riftanomalysystemconfigdata_chanceofspawnsafezonebasedanomalies_71_0941dfcb4c5215d20a9246a310499acb) (Utilities|Struct|BreakRiftAnomaliesAssetRegistrySystemData (Variables|GetRiftAnomalySystemConfigData_PerStormCircleGeneralSpawnChance_26_AAECEAA646D26E24C4340F8A66FFBA3B)))
  (if (>= _riftanomalysystemconfigdata_chanceofspawnsafezonebasedanomalies_71_0941dfcb4c5215d20a9246a310499acb (Math|Random|RandomFloatinRange 0.0 1.0))
    (return true)
    (else
      (return false))))

(fn IsSafeZoneFree
  (params (SafeZoneIndexToCheck :type Integer))
  (bind _output_anomaliesdataassets_20_7e8609f440fac2369e8ac7902781fa79 (Utilities|Struct|BreakRiftAnomaliesAssetRegistryData (Utilities|Array|Get(acopy) (Variables|DataRegistry|GetPredefinedCircleAnomaliesDrops) (Variables|GetSafeZoneIndexToCheck))))
  (if (Utilities|Array|IsValidIndex (Variables|DataRegistry|GetPredefinedCircleAnomaliesDrops) (Variables|GetSafeZoneIndexToCheck))
    (Utilities|IsValid _output_anomaliesdataassets_20_7e8609f440fac2369e8ac7902781fa79
      (:Is%20Valid
        (return false))
      (:Is%20Not%20Valid
        (return true)))
    (else
      (Fortnite|FortShippingLog self "LogFortRAManager: Trying to search on an invalid safezone index"))))

(fn IsSafeZoneSpawnActive
  (return (Variables|Default|GetbSafeZoneSpawnActive)))

(fn IsSystemEnabled
  (bind (_riftanomalysystemconfigdata_perstormcirclegeneralspawnchance_26_aaeceaa646d26e24c4340f8a66ffba3b _riftanomalysystemconfigdata_minsafezone_40_6abc48b2422b44b585ba67a5f74173a6 _riftanomalysystemconfigdata_maxsafezone_41_f34071834b0d86641a5860a838c2db3c _riftanomalysystemconfigdata_predictionwindow_37_4ad963b14b0f959ff7c44898407c2348 _riftanomalysystemconfigdata_cananomaliesrepeat_44_6ce45a574f0d808a542d9fb5e0603294 _riftanomalysystemconfigdata_rasystemenabled_46_fdecd3964c50a8aaab93158bbb40a992) (Utilities|Struct|BreakRiftAnomaliesAssetRegistrySystemData (Variables|GetRiftAnomalySystemConfigData_PerStormCircleGeneralSpawnChance_26_AAECEAA646D26E24C4340F8A66FFBA3B)))
  (return _riftanomalysystemconfigdata_rasystemenabled_46_fdecd3964c50a8aaab93158bbb40a992))

(fn IsUsingChanceByAnomalyMode
  (bind (_riftanomalysystemconfigdata_perstormcirclegeneralspawnchance_26_aaeceaa646d26e24c4340f8a66ffba3b _riftanomalysystemconfigdata_minsafezone_40_6abc48b2422b44b585ba67a5f74173a6 _riftanomalysystemconfigdata_maxsafezone_41_f34071834b0d86641a5860a838c2db3c _riftanomalysystemconfigdata_predictionwindow_37_4ad963b14b0f959ff7c44898407c2348 _riftanomalysystemconfigdata_cananomaliesrepeat_44_6ce45a574f0d808a542d9fb5e0603294 _riftanomalysystemconfigdata_rasystemenabled_46_fdecd3964c50a8aaab93158bbb40a992 _riftanomalysystemconfigdata_decreaseweightofanomalyinarow_50_39fec66249842b8db450b59845657a63 _riftanomalysystemconfigdata_maxsafezonespermatch_107_9a5f555f4dd71278a52e6b952b5e362a _riftanomalysystemconfigdata_overridestormweightbyphaserowbytag_60_9a5ba2b34503c08e68ce4f841220589b _riftanomalysystemconfigdata_chanceofspawnsafezonebasedanomalies_71_0941dfcb4c5215d20a9246a310499acb _riftanomalysystemconfigdata_anomalysubsetoverriding_70_82b0e46f433a27407a9f93a568955f28 _riftanomalysystemconfigdata_usechancebyanomalysystemmode_84_c460390f49dae179dabf328c947469ec) (Utilities|Struct|BreakRiftAnomaliesAssetRegistrySystemData (Variables|GetRiftAnomalySystemConfigData_PerStormCircleGeneralSpawnChance_26_AAECEAA646D26E24C4340F8A66FFBA3B)))
  (return _riftanomalysystemconfigdata_usechancebyanomalysystemmode_84_c460390f49dae179dabf328c947469ec))

(fn ListenForPawnSpawnForDeadPlayers
  (for _array_element (Fortnite|GetAllFortPlayerControllers true)
    (Class|FortQuestBlueprintLibrary|GetFortPawnFromObject _array_element
      (:NotValid
        (CallFunction|BindtoPlayerPawnSpawn _array_element)))))

(fn OnRep_ActiveRA
  ;[1 1 1] Client Async Load
  (if (not (Networking|IsDedicatedServer))
    (Utilities|IsValid (Variables|Default|GetActiveRA)
      (:Is%20Valid
        (Variables|Default|SetLastLoadedActiveRA (Variables|Default|GetActiveRA)))
      (:Is%20Not%20Valid
        (if (Utilities|IsValid (Variables|Default|GetLastLoadedActiveRA) :!class /Script/Engine.KismetSystemLibrary)
          (Utilities|IsValid (Variables|Default|GetIncomingRA)
            (:Is%20Valid
              (if (== (Variables|Default|GetIncomingRA) (Variables|Default|GetLastLoadedActiveRA))
                _
                (else
                  _)))
            (:Is%20Not%20Valid
              _))
          (CallFunction|ResetAssetHandle (Variables|Default|GetLastLoadedActiveRA)))))))

(fn OnRep_IncomingRA
  (bind _incomingra (Variables|Default|GetIncomingRA))
  (if (not (Networking|IsDedicatedServer))
    (Utilities|IsValid _incomingra
      (:Is%20Valid
        (CallFunction|LoadAssetsfromRADataAsset _incomingra)))))

(fn OnRep_MatchStartAnomalyData
  (switch Utilities|FlowControl|Switch|SwitchonBP_Enum_RA_UI_STATE (Variables|ui|GetRiftAnomaliesUIState)
    (:NewEnumerator0
      (Variables|ui|SetRiftAnomaliesUIState "NewEnumerator4"))
    (:NewEnumerator1
      (UI|CheckifCanActivateStartMatchAnomalyUI))))

(fn PerformEQS
  (bind _currentsafezoneinfo (Variables|Default|GetCurrentSafeZoneInfo))
  (bind (_center _radius) (Utilities|Struct|BreakFortSafeZonePhaseInfo _currentsafezoneinfo))
  (bind _returnvalue (AI|EQS|RunEQSQuery (Class|FortRiftAnomaliesData|GetConfiguredEQS (Variables|Default|GetIncomingRA)) self))
  (AI|EQS|BindEventtoOnQueryFinished _returnvalue (EventDispatchers|CreateEvent :!event OnQueryFinished))
  (AI|EQS|SetNamedParam _returnvalue "GridSize" _radius)
  (AI|EQS|SetNamedParam _returnvalue "WorldLocationParamX" (.x _center))
  (AI|EQS|SetNamedParam _returnvalue "WorldLocationParamY" (.y _center))
  (AI|EQS|SetNamedParam _returnvalue "WorldLocationParamZ" (.z _center)))

(fn PlayRAEndSound
  (params (Sound :type Sound%20Base%20Object%20Reference))
  (Networking|CanExecuteCosmeticEvents
    (:True
      (Utilities|IsValid Sound
        (:Is%20Valid
          (if (AudioAssetUserDataTags|HasTag Sound "(TagName=\"Audio.Metadata.Music\")")
            (Audio|AddEvent (GameInstanceSubsystems|GetMusicEventSubsystem) Sound "(TagName=\"MusicEventSubsystem.Event.Gameplay.Ability.1P\")" "(GameplayTags=)")
            (else
              (Audio|PlaySound2D Sound))))))))

(fn PlayRATriggerSounds
  (params (AudioToPlay :type Sound%20Base%20Object%20Reference))
  (Networking|CanExecuteCosmeticEvents
    (:True
      (Utilities|IsValid AudioToPlay
        (:Is%20Valid
          (if (AudioAssetUserDataTags|HasTag AudioToPlay "(TagName=\"Audio.Metadata.Music\")")
            (Audio|AddEvent (GameInstanceSubsystems|GetMusicEventSubsystem) AudioToPlay "(TagName=\"MusicEventSubsystem.Event.Gameplay.Ability.1P\")" "(GameplayTags=)")
            (else
              (Audio|PlaySound2D AudioToPlay))))))))

(fn PreDefiningAllCircleAnomaliesOnAnomalyWeightedMode
  (bind (_riftanomalysystemconfigdata_perstormcirclegeneralspawnchance_26_aaeceaa646d26e24c4340f8a66ffba3b _riftanomalysystemconfigdata_minsafezone_40_6abc48b2422b44b585ba67a5f74173a6 _riftanomalysystemconfigdata_maxsafezone_41_f34071834b0d86641a5860a838c2db3c _riftanomalysystemconfigdata_predictionwindow_37_4ad963b14b0f959ff7c44898407c2348 _riftanomalysystemconfigdata_cananomaliesrepeat_44_6ce45a574f0d808a542d9fb5e0603294 _riftanomalysystemconfigdata_rasystemenabled_46_fdecd3964c50a8aaab93158bbb40a992 _riftanomalysystemconfigdata_decreaseweightofanomalyinarow_50_39fec66249842b8db450b59845657a63 _riftanomalysystemconfigdata_maxsafezonespermatch_107_9a5f555f4dd71278a52e6b952b5e362a) (Utilities|Struct|BreakRiftAnomaliesAssetRegistrySystemData (Variables|GetRiftAnomalySystemConfigData_PerStormCircleGeneralSpawnChance_26_AAECEAA646D26E24C4340F8A66FFBA3B)))
  ;[1 1 1] First Implementation
  (for _index (range (+ _riftanomalysystemconfigdata_maxsafezonespermatch_107_9a5f555f4dd71278a52e6b952b5e362a 1))
    (LocalVariables|SetLoopIndex _index)
    (bind _isstormcirclevalidforanomaly (CallFunction|IsGivenStormCircleHaveAnomalies _index))
    (if _isstormcirclevalidforanomaly
      (bind (_radata _success) (Determinism|GetRandomizedRAConsideringWeights true (LocalVariables|GetLoopIndex) (LocalVariables|GetLastSortedAnomaly)))
      (if _success
        (bind _output_get (LocalVariables|SetLastSortedAnomaly (.RowName_5_4E12E4C6414E627E8EC7CE9FD91AAD35 (Utilities|Struct|BreakBRADataRegistrywithRowNameStruct _radata))))
        (RAModifiers|RemoveRAModifiers (.RATag_47_3BB6E306408E53230FE68FAC1BCD5BB3 (Utilities|Struct|BreakRiftAnomaliesAssetRegistryData (Utilities|Struct|BreakBRADataRegistrywithRowNameStruct _radata))))
        (RAModifiers|HandleAddingNewRAModifiers (.RAModfierMap_59_87219A6747987167EB1CC3945BD39CF7 (Utilities|Struct|BreakRiftAnomaliesAssetRegistryData (Utilities|Struct|BreakBRADataRegistrywithRowNameStruct _radata))))
        ;;[1 1 1] The server will load all predetermined anomalies
        (bind _anomaliedata_2_f38e3dd0464ff4283aff31b4e94803e4 (Utilities|Struct|BreakBRADataRegistrywithRowNameStruct _radata))
        (bind _anomaliedata_2_f38e3dd0464ff4283aff31b4e94803e4_anomaliesdataassets_20_7e8609f440fac2369e8ac7902781fa79 (Utilities|Struct|BreakRiftAnomaliesAssetRegistryData _anomaliedata_2_f38e3dd0464ff4283aff31b4e94803e4))
        (CallFunction|LoadAssetsfromRADataAsset _anomaliedata_2_f38e3dd0464ff4283aff31b4e94803e4_anomaliesdataassets_20_7e8609f440fac2369e8ac7902781fa79)
        (Utilities|Array|SetArrayElem (Variables|DataRegistry|GetPredefinedCircleAnomaliesDrops) (LocalVariables|GetLoopIndex) (Utilities|Struct|BreakBRADataRegistrywithRowNameStruct _radata) true)
        (Development|PrintString (Utilities|String|Append (Utilities|String|ToString(Name) _output_get) "   WIll spawn at circle: " (Utilities|String|ToString(Integer) (LocalVariables|GetLoopIndex))) true true "(R=0.000000,G=0.660000,B=1.000000,A=1.000000)" 2e+01)
        (else
          (Development|PrintString "NO ANOMALY RANDOMED")))
      (else
        (Development|PrintString (Utilities|String|Append "LogFortRAManager: " " Storm circle index: " (Utilities|String|ToString(Integer) (LocalVariables|GetLoopIndex)) " will not spawn any anomaly by setting")))))
  (Development|PrintString "FInish anomaly setup"))

(fn PreDefiningAllCircleAnomaliesOnSafeZoneWeightedMode
  (bind (_riftanomalysystemconfigdata_perstormcirclegeneralspawnchance_26_aaeceaa646d26e24c4340f8a66ffba3b _riftanomalysystemconfigdata_minsafezone_40_6abc48b2422b44b585ba67a5f74173a6 _riftanomalysystemconfigdata_maxsafezone_41_f34071834b0d86641a5860a838c2db3c _riftanomalysystemconfigdata_predictionwindow_37_4ad963b14b0f959ff7c44898407c2348 _riftanomalysystemconfigdata_cananomaliesrepeat_44_6ce45a574f0d808a542d9fb5e0603294 _riftanomalysystemconfigdata_rasystemenabled_46_fdecd3964c50a8aaab93158bbb40a992 _riftanomalysystemconfigdata_decreaseweightofanomalyinarow_50_39fec66249842b8db450b59845657a63 _riftanomalysystemconfigdata_maxsafezonespermatch_107_9a5f555f4dd71278a52e6b952b5e362a _riftanomalysystemconfigdata_overridestormweightbyphaserowbytag_60_9a5ba2b34503c08e68ce4f841220589b _riftanomalysystemconfigdata_chanceofspawnsafezonebasedanomalies_71_0941dfcb4c5215d20a9246a310499acb _riftanomalysystemconfigdata_anomalysubsetoverriding_70_82b0e46f433a27407a9f93a568955f28 _riftanomalysystemconfigdata_usechancebyanomalysystemmode_84_c460390f49dae179dabf328c947469ec _riftanomalysystemconfigdata_weightsforquantityofanomaliesinamatch_97_7eecd97b46b44ef99cd0db914f6f2dcd _riftanomalysystemconfigdata_weightofananomalytobesortedonweightedsafezones_102_c579e04941f00bb8e72f949d6856433b _riftanomalysystemconfigdata_defaultweightforanomalyselection_106_86131a7644fbc538c8a7c7911131ce71 _riftanomalysystemconfigdata_defaultsafezoneweights_110_d4f62c4d4b770b30b08cf3acd19dd541 _riftanomalysystemconfigdata_prioritizeanomalywhiledecidingsafezones_114_9dbf0cf64512cfb0f4864b9ecdd77549) (Utilities|Struct|BreakRiftAnomaliesAssetRegistrySystemData (Variables|GetRiftAnomalySystemConfigData_PerStormCircleGeneralSpawnChance_26_AAECEAA646D26E24C4340F8A66FFBA3B)))
  (bind (_riftanomalysystemconfigdata_perstormcirclegeneralspawnchance_26_aaeceaa646d26e24c4340f8a66ffba3b_2 _riftanomalysystemconfigdata_minsafezone_40_6abc48b2422b44b585ba67a5f74173a6_2 _riftanomalysystemconfigdata_maxsafezone_41_f34071834b0d86641a5860a838c2db3c_2 _riftanomalysystemconfigdata_predictionwindow_37_4ad963b14b0f959ff7c44898407c2348_2 _riftanomalysystemconfigdata_cananomaliesrepeat_44_6ce45a574f0d808a542d9fb5e0603294_2 _riftanomalysystemconfigdata_rasystemenabled_46_fdecd3964c50a8aaab93158bbb40a992_2 _riftanomalysystemconfigdata_decreaseweightofanomalyinarow_50_39fec66249842b8db450b59845657a63_2 _riftanomalysystemconfigdata_maxsafezonespermatch_107_9a5f555f4dd71278a52e6b952b5e362a_2 _riftanomalysystemconfigdata_overridestormweightbyphaserowbytag_60_9a5ba2b34503c08e68ce4f841220589b_2 _riftanomalysystemconfigdata_chanceofspawnsafezonebasedanomalies_71_0941dfcb4c5215d20a9246a310499acb_2 _riftanomalysystemconfigdata_anomalysubsetoverriding_70_82b0e46f433a27407a9f93a568955f28_2 _riftanomalysystemconfigdata_usechancebyanomalysystemmode_84_c460390f49dae179dabf328c947469ec_2 _riftanomalysystemconfigdata_weightsforquantityofanomaliesinamatch_97_7eecd97b46b44ef99cd0db914f6f2dcd_2 _riftanomalysystemconfigdata_weightofananomalytobesortedonweightedsafezones_102_c579e04941f00bb8e72f949d6856433b_2 _riftanomalysystemconfigdata_defaultweightforanomalyselection_106_86131a7644fbc538c8a7c7911131ce71_2 _riftanomalysystemconfigdata_defaultsafezoneweights_110_d4f62c4d4b770b30b08cf3acd19dd541_2 _riftanomalysystemconfigdata_prioritizeanomalywhiledecidingsafezones_114_9dbf0cf64512cfb0f4864b9ecdd77549_2 _riftanomalysystemconfigdata_anomaliesgroupsconfiguration_118_57b7639f457cde58dce1cfa3ccb50010) (Utilities|Struct|BreakRiftAnomaliesAssetRegistrySystemData (Variables|GetRiftAnomalySystemConfigData_PerStormCircleGeneralSpawnChance_26_AAECEAA646D26E24C4340F8A66FFBA3B)))
  (bind _selectedanomaly_anomaliesdataassets_20_7e8609f440fac2369e8ac7902781fa79 (Utilities|Struct|BreakRiftAnomaliesAssetRegistryData (Variables|GetSelectedAnomaly_AnomaliesDataAssets_20_7E8609F440FAC2369E8AC7902781FA79)))
  (bind _selectedanomaly_anomaliesdataassets_20_7e8609f440fac2369e8ac7902781fa79_2 (Utilities|Struct|BreakRiftAnomaliesAssetRegistryData (Variables|GetSelectedAnomaly_AnomaliesDataAssets_20_7E8609F440FAC2369E8AC7902781FA79)))
  (bind _quantityofanomalies (LocalVariables|GetQuantityOfAnomalies))
  ;[1 1 1] First Implementation
  (bind _output_get (LocalVariables|SetGroupsOfAnomalies _riftanomalysystemconfigdata_anomaliesgroupsconfiguration_118_57b7639f457cde58dce1cfa3ccb50010))
  (if (> (Utilities|Array|Length _output_get) 0)
    (for (_array_element _array_index) (LocalVariables|GetGroupsOfAnomalies)
      (bind _array_element_weightsforquantityofanomaliesinamatch (Utilities|Struct|BreakFortRAAnomaliesGroupConfig _array_element))
      (LocalVariables|SetAnomalyGroupConfigIndex _array_index)
      (bind _quantityforthematch (Determinism|SetMatchQuantityOfAnomalies _array_element_weightsforquantityofanomaliesinamatch))
      (LocalVariables|SetQuantityOfAnomalies _quantityforthematch)
      (if (> _quantityofanomalies 0)
        (bind _weightedralist (Determinism|SelectAnomaliesbyWeight _quantityofanomalies (Utilities|Array|Get(acopy) (LocalVariables|GetGroupsOfAnomalies) (LocalVariables|GetAnomalyGroupConfigIndex))))
        (bind _output_get_2 (LocalVariables|SetWeightedAnomalySelectedList _weightedralist))
        (if _riftanomalysystemconfigdata_prioritizeanomalywhiledecidingsafezones_114_9dbf0cf64512cfb0f4864b9ecdd77549
          (bind _selectedanomaliesbysafezone (Determinism|SelectAnomaliesSafezonesPrioritizingAnomalies _output_get_2 (Utilities|Array|Get(acopy) (LocalVariables|GetGroupsOfAnomalies) (LocalVariables|GetAnomalyGroupConfigIndex))))
          (bind (_keypin _valuepin) (Utilities|Map|ForEachLoop(Map) _selectedanomaliesbysafezone))
          (LocalVariables|SetCurrentIteratedSafeZoneIndex _keypin)
          (LocalVariables|SetCurrentIteratedTag _valuepin)
          (else
            (bind _selectedanomaliesbysafezone_2 (Determinism|SelectedAnomaliesSafeZonesBasedonWeight _output_get_2))
            (for (_array_element_2 _array_index_2) _selectedanomaliesbysafezone_2
              (LocalVariables|SetCurrentIteratedSafeZoneIndex _array_index_2)
              (LocalVariables|SetCurrentIteratedTag _array_element_2)
              (if (GameplayTags|IsGameplayTagValid (LocalVariables|GetCurrentIteratedTag))
                (bind (_anomaly _sucess) (DataRegistry|FindDataRegistrybyTag (LocalVariables|GetCurrentIteratedTag)))
                (if _sucess
                  (bind _output_get_3 (LocalVariables|SetSelectedAnomaly _anomaly))
                  (Utilities|Array|SetArrayElem (Variables|DataRegistry|GetPredefinedCircleAnomaliesDrops) (LocalVariables|GetCurrentIteratedSafeZoneIndex) _output_get_3 true)
                  (Fortnite|FortShippingLog self (Utilities|String|Append (Utilities|String|ToString(Text) (Class|FortRiftAnomaliesData|GetRAName _selectedanomaly_anomaliesdataassets_20_7e8609f440fac2369e8ac7902781fa79 :!class /Script/RiftAnomaliesGameplayRuntime.FortRiftAnomaliesData)) "LogFortRAManager:  WIll spawn at weighted safezone: " (Utilities|String|ToString(Integer) (LocalVariables|GetCurrentIteratedSafeZoneIndex)) "Anomalies Group Index Is:" (Utilities|String|ToString(Integer) (LocalVariables|GetAnomalyGroupConfigIndex))))
                  ;; Asyncload
                  (CallFunction|LoadAssetsfromRADataAsset _selectedanomaly_anomaliesdataassets_20_7e8609f440fac2369e8ac7902781fa79_2)
                  (else
                    (Fortnite|FortShippingLog self "LogFortRAManager: No tag defined for anomaly, please review registered data assets and assign the anomalies tags" true)))))))
        (else
          (Fortnite|FortShippingLog self "LogFortRAManager: No Anomaly Selected for the match..."))))))

(fn RAActivatedClientUIActions
  (params (InputPin :type Fort%20Rift%20Anomalies%20Data%20Object%20Reference))
  (Networking|CanExecuteCosmeticEvents
    (:True
      (switch Utilities|FlowControl|Switch|SwitchonBP_Enum_RA_UI_STATE (Variables|ui|GetRiftAnomaliesUIState)
        (:NewEnumerator0
          (Variables|ui|SetRiftAnomaliesUIState "NewEnumerator4"))
        (:NewEnumerator4
          _)
        (:NewEnumerator1
          _)
        (:NewEnumerator5
          _))
      (Events|CallOnRiftAnomalyActivated InputPin))))

(fn RAClearActiveClientUIActions
  (Events|CallOnClearAnomaly))

(fn RAEndedWithNotificationClientUIActions
  (params (RiftAnomalyData :type Fort%20Rift%20Anomalies%20Data%20Object%20Reference))
  (Networking|CanExecuteCosmeticEvents
    (:True
      (Events|CallOnRiftAnomalyEndedwithNotification RiftAnomalyData))))

(fn RAIncomingClientUIActions
  (params (InputPin :type Fort%20Rift%20Anomalies%20Data%20Object%20Reference))
  (Networking|CanExecuteCosmeticEvents
    (:True
      (Events|CallOnRiftAnomalySpawnAnticipation InputPin))))

(fn RemoveAnomalyFromPossibleList
  (params (RAName :type Name) (RATag :type Gameplay%20Tag%20Structure))
  (for (_array_element _array_index) (Variables|DataRegistry|GetPossibleAnomaliesToSpawn)
    (bind (_array_element_anomaliedata_2_f38e3dd0464ff4283aff31b4e94803e4 _array_element_rowname_5_4e12e4c6414e627e8ec7ce9fd91aad35) (Utilities|Struct|BreakBRADataRegistryWithRowNameStruct _array_element))
    (if (== _array_element_rowname_5_4e12e4c6414e627e8ec7ce9fd91aad35 (GameplayTags|GetTagName RATag))
      (LocalVariables|SetIndexToRemove _array_index)))
  (Utilities|Array|RemoveIndex (Variables|DataRegistry|GetPossibleAnomaliesToSpawn) (LocalVariables|GetIndexToRemove)))

(fn RemoveRAEffects
  (params (Pawn :type Fort%20Pawn%20Object%20Reference) (RAData :type Fort%20Rift%20Anomalies%20Data%20Object%20Reference))
  (bind _owningcontroller (LocalVariables|GetOwningController))
  (bind _abilitysethandles (Variables|Default|GetAbilitySetHandles))
  (bind (_value _returnvalue) (Utilities|Map|Find _abilitysethandles _owningcontroller))
  (bind _output_get (LocalVariables|SetOwningController (Pawn|GetController Pawn)))
  (Utilities|IsValid _output_get
    (:Is%20Valid
      (switch Utilities|FlowControl|Switch|SwitchonERAEffectsRules (Class|FortRiftAnomaliesData|GetEffectApplicationRules RAData)
        (:ApplyToAllPlayers
          (if _returnvalue
            (Ability|UnequipFortAbilitySet _value)
            (Utilities|Map|Remove (Variables|Default|GetAbilitySetHandles) (LocalVariables|GetOwningController))
            (CallFunction|UnbindfromPlayerPawnSpawn (LocalVariables|GetOwningController))))))))

(fn RemoveRAModifiers
  (params (RATagToRemoveModifiersFrom :type Gameplay%20Tag%20Structure))
  (if (Utilities|Map|Contains (Variables|RAModifiers|GetRAActiveModifiers) RATagToRemoveModifiersFrom)
    (Utilities|Map|Remove (Variables|RAModifiers|GetRAActiveModifiers) (Variables|GetRATagToRemoveModifiersFrom))))

(fn ReturnRAFromRowName
  (params (RowName :type Name))
  (bind _allanomaliesdata (Variables|DataRegistry|GetAllAnomaliesData))
  (if (> (Utilities|Array|Length _allanomaliesdata) 0)
    (for (_array_element _array_index) _allanomaliesdata
      (bind (_array_element_anomaliedata_2_f38e3dd0464ff4283aff31b4e94803e4 _array_element_rowname_5_4e12e4c6414e627e8ec7ce9fd91aad35) (Utilities|Struct|BreakBRADataRegistryWithRowNameStruct _array_element))
      (if (== _array_element_rowname_5_4e12e4c6414e627e8ec7ce9fd91aad35 (Variables|GetRowName))
        (Utilities|IsValid (Utilities|Struct|BreakRiftAnomaliesAssetRegistryData _array_element_anomaliedata_2_f38e3dd0464ff4283aff31b4e94803e4)
          (:Is%20Valid
            (bind _output_anomaliedata_2_f38e3dd0464ff4283aff31b4e94803e4 (Utilities|Struct|BreakBRADataRegistryWithRowNameStruct (Utilities|Array|Get(acopy) (Variables|DataRegistry|GetAllAnomaliesData) _array_index)))
            (return (Utilities|Struct|BreakRiftAnomaliesAssetRegistryData _output_anomaliedata_2_f38e3dd0464ff4283aff31b4e94803e4) _output_anomaliedata_2_f38e3dd0464ff4283aff31b4e94803e4)))))
    (return 0 0)
    (else
      (return 0 0))))

(fn RiftAnomalyDebugForceEndActive
  (Utilities|SwitchHasAuthority
    (:Authority
      (Debug|ForceEndActiveRiftAnomalies false))))

(fn RiftAnomalyDumpAvailable
  (bind _namesarray (LocalVariables|GetNamesArray))
  (bind _info (LocalVariables|GetInfo))
  (Networking|SwitchHasAuthority
    (:Authority
      (LocalVariables|SetInfo "--------------
Available Rift Anomalies:")
      (Utilities|FlowControl|Sequence
        (:then_0
          (bind _outidlist (DataRegistry|GetPossibleDataRegistryIdList "AnomaliesToLoad"))
          (for _array_element _outidlist
            (bind _outitem (DataRegistry|FindDataRegistryItem _array_element)
              (:Found
                (Utilities|Array|Add (LocalVariables|GetNamesArray) (.RowName_5_4E12E4C6414E627E8EC7CE9FD91AAD35 (Utilities|Struct|BreakBRADataRegistrywithRowNameStruct (Utilities|Struct|MakeBRADataRegistrywithRowNameStruct _outitem (.ItemName (Utilities|Struct|BreakDataRegistryId _array_element))))))))))
        (:then_1
          ;[1 1 1] Alphabetize Sort
          (Utilities|Array|Sort|SortNameArray _namesarray)
          (for _array_element_2 _namesarray
            (LocalVariables|SetInfo (Utilities|String|ToString(Text) (Utilities|Text|FormatText "NSLOCTEXT(\"[4B9D47DCB44E2CC832FF8DB0E2DA666C]\", \"CCC89B104E1EACFDDF2B81995D44DE83\", \"{Info}\\r\\n{AnomalyRowName} \")" (LocalVariables|GetInfo) _array_element_2))))
          (LocalVariables|SetInfo (Utilities|String|ToString(Text) (Utilities|Text|FormatText "NSLOCTEXT(\"[4B9D47DCB44E2CC832FF8DB0E2DA666C]\", \"D4931A0B44A25A59510043BC8F9867AC\", \"{Info}\\r\\n--------------\")" (LocalVariables|GetInfo)))))
        (:then_2
          (Fortnite|FortShippingLog self _info)
          (Development|PrintString _info true false "(R=0.000000,G=0.660000,B=1.000000,A=1.000000)" 1e+01)
          (bind _outfortplayerpawns (Fortnite|GetFortPlayerPawns))
          (for _array_element_3 _outfortplayerpawns
            (CallFunction|SendClientMessage (Fortnite|GetFortPlayerControllerfromActor _array_element_3) (LocalVariables|GetInfo))))))))

(fn RiftAnomalyDumpInfo
  (bind _info (LocalVariables|GetInfo))
  (Networking|SwitchHasAuthority
    (:Authority
      (LocalVariables|SetInfo "--------")
      (Utilities|FlowControl|Sequence
        (:then_0
          (if (Utilities|Array|IsNotEmpty (Variables|Default|GetAllActiveRAInfo))
            (for _array_element (Variables|Default|GetAllActiveRAInfo)
              (bind _array_element_activeradata_6_893472d943ba8730134436b6a7e04009 (Utilities|Struct|BreakBRAActiveInfoStruct _array_element))
              (LocalVariables|SetInfo (Utilities|String|ToString(Text) (Utilities|Text|FormatText "NSLOCTEXT(\"[4B9D47DCB44E2CC832FF8DB0E2DA666C]\", \"A670332C4838C4F8D586088D9F506595\", \"{Info} \\r\\nActive: {RiftAnomalyName}\")" (LocalVariables|GetInfo) (Variables|/Script/RiftAnomaliesGameplayRuntime.FortRiftAnomaliesData|GetRAName _array_element_activeradata_6_893472d943ba8730134436b6a7e04009)))))
            (else
              (LocalVariables|SetInfo (Utilities|String|ToString(Text) (Utilities|Text|FormatText "NSLOCTEXT(\"[4B9D47DCB44E2CC832FF8DB0E2DA666C]\", \"3DB1B5CE4244BB259BE677A71BA7EC59\", \"{Info} \\r\\nActive: None\")" (LocalVariables|GetInfo)))))))
        (:then_1
          (bind _incomingra (Variables|Default|GetIncomingRA)
            (:then
              (LocalVariables|SetInfo (Utilities|String|ToString(Text) (Utilities|Text|FormatText "NSLOCTEXT(\"[4B9D47DCB44E2CC832FF8DB0E2DA666C]\", \"FD47A2374B54652426451EBF1E701EB3\", \"{Info} \\r\\nIncoming: {RiftAnomalyName}\")" (LocalVariables|GetInfo) (Variables|/Script/RiftAnomaliesGameplayRuntime.FortRiftAnomaliesData|GetRAName _incomingra)))))
            (:else
              (LocalVariables|SetInfo (Utilities|String|ToString(Text) (Utilities|Text|FormatText "NSLOCTEXT(\"[4B9D47DCB44E2CC832FF8DB0E2DA666C]\", \"5303D120487E6893FAE9E0901C3BD388\", \"{Info} \\r\\nNo incoming rift anomaly\")" (LocalVariables|GetInfo)))))))
        (:then_2
          (bind _returnvalue (SafeZone|GetCurrentSafeZonePhase))
          (LocalVariables|SetInfo (Utilities|String|ToString(Text) (Utilities|Text|FormatText "NSLOCTEXT(\"[4B9D47DCB44E2CC832FF8DB0E2DA666C]\", \"24C229704ADC427C2A6BE5BB07C80057\", \"{Info} \\r\\n----------\\r\\nCurrent Storm Circle: {StormIndex}\")" (LocalVariables|GetInfo) _returnvalue))))
        (:then_3
          (for (_array_element_2 _array_index) (Variables|DataRegistry|GetPredefinedCircleAnomaliesDrops)
            (Utilities|IsValid (Utilities|Struct|BreakRiftAnomaliesAssetRegistryData _array_element_2)
              (:Is%20Valid
                (LocalVariables|SetInfo (Utilities|String|ToString(Text) (Utilities|Text|FormatText "NSLOCTEXT(\"[4B9D47DCB44E2CC832FF8DB0E2DA666C]\", \"15BAEA934265EC317B362993C25CE729\", \"{Info} \\r\\nPredetermined Order: {RiftAnomalyName} on Storm Circle {StormIndex}\")" (LocalVariables|GetInfo) (Variables|/Script/RiftAnomaliesGameplayRuntime.FortRiftAnomaliesData|GetRAName (Utilities|Struct|BreakRiftAnomaliesAssetRegistryData _array_element_2)) _array_index)))))))
        (:then_4
          (Fortnite|FortShippingLog self _info)
          (Development|PrintString _info true false "(R=0.000000,G=0.660000,B=1.000000,A=1.000000)" 1e+01)
          (bind _outfortplayerpawns (Fortnite|GetFortPlayerPawns))
          (for _array_element_3 _outfortplayerpawns
            (CallFunction|SendClientMessage (Fortnite|GetFortPlayerControllerfromActor _array_element_3) (LocalVariables|GetInfo))))))))

(fn RiftAnomalyForceSpawn
  (params (Name :type String))
  (bind _currentsafezoneinfo_center (Utilities|Struct|BreakFortSafeZonePhaseInfo (Variables|GetCurrentSafeZoneInfo_Center)))
  (Utilities|SwitchHasAuthority
    (:Authority
      (bind _riftanomaliedata (CallFunction|ReturnRAFromRowName (Utilities|String|StringToName (Variables|GetName))))
      (bind _output_get (LocalVariables|SetAnomalyData _riftanomaliedata))
      (Utilities|IsValid _output_get
        (:Is%20Valid
          (Debug|RiftAnomalyDebugForceEndActive)
          (bind _output_get_2 (Variables|Debug|SetDebugRAToSpawn _output_get))
          (if (Class|FortRiftAnomaliesData|GetIsUsingUseEQS _output_get_2)
            (Debug|DebugForceSpawnPerformEQS)
            (else
              (bind _output_get_3 (Variables|Debug|SetDebugSpawnLocation _currentsafezoneinfo_center))
              (Debug|DebugForceSpawnRA _output_get_3 (Variables|Debug|GetDebugRAToSpawn)))))))))

(fn RiftAnomalyForceSpawnRandom
  (bind _returnvalue (SafeZone|GetSafeZoneIndicator))
  (bind (_radata _success) (Determinism|GetRandomizedRA false))
  (if _success
    (Debug|RiftAnomalyDebugForceEndActive)
    (Utilities|IsValid _returnvalue
      (:Is%20Valid
        (Spawning|TrySpawnRiftAnomaly (SafeZone|GetSafeZoneCenter _returnvalue) (Utilities|Struct|BreakRiftAnomaliesAssetRegistryData _radata)))
      (:Is%20Not%20Valid
        (bind _outactors (Actor|GetAllActorsOfClass "/Script/FortniteGame.FortPlayerPawnAthena"))
        (Spawning|TrySpawnRiftAnomaly (Transformation|GetActorLocation (Utilities|Array|Get(acopy) _outactors)) (Utilities|Struct|BreakRiftAnomaliesAssetRegistryData _radata))))))

(fn RiftAnomalySetNext
  (params (Name :type String))
  (Utilities|SwitchHasAuthority
    (:Authority
      (if (not (Utilities|String|IsEmpty (Variables|GetName)))
        (bind _riftanomaliedata (CallFunction|ReturnRAFromRowName (Utilities|String|StringToName (Variables|GetName))))
        (Utilities|IsValid _riftanomaliedata
          (:Is%20Valid
            (Utilities|Struct|SetmembersinRiftAnomaliesAssetRegistryData (Utilities|Array|Get(aref) (Variables|DataRegistry|GetPredefinedCircleAnomaliesDrops) (+ (SafeZone|GetCurrentPhaseIndex (SafeZone|GetSafeZoneIndicator)) 1)) _riftanomaliedata)
            (Prediction|SetupNextPhaseRA)))))))

(fn RiftAnomalySetSpawnOrder
  (params (Name :type String))
  (bind _predefinedcircleanomaliesdrops (Variables|DataRegistry|GetPredefinedCircleAnomaliesDrops))
  (bind _predefinedcircleanomaliesdrops_2 (Variables|DataRegistry|GetPredefinedCircleAnomaliesDrops))
  (Utilities|SwitchHasAuthority
    (:Authority
      (if (not (Utilities|String|IsEmpty (Variables|GetName)))
        (Utilities|FlowControl|Sequence
          (:then_0
            ;[1 1 1] Clear existing data
            (for (_ _array_index) _predefinedcircleanomaliesdrops
              (Utilities|Struct|SetmembersinRiftAnomaliesAssetRegistryData (Utilities|Array|Get(aref) _predefinedcircleanomaliesdrops _array_index) 0)))
          (:then_1
            (for (_array_element _array_index_2) (Utilities|String|ParseIntoArray (Variables|GetName) ", ")
              (if (Utilities|String|NotEqual,CaseInsenstive(String) _array_element "none")
                (bind _riftanomaliedata (CallFunction|ReturnRAFromRowName (Utilities|String|StringToName _array_element)))
                (Utilities|IsValid _riftanomaliedata
                  (:Is%20Valid
                    (Utilities|Struct|SetmembersinRiftAnomaliesAssetRegistryData (Utilities|Array|Get(aref) _predefinedcircleanomaliesdrops_2 _array_index_2) _riftanomaliedata))
                  (:Is%20Not%20Valid
                    (Utilities|Struct|SetmembersinRiftAnomaliesAssetRegistryData (Utilities|Array|Get(aref) _predefinedcircleanomaliesdrops_2 _array_index_2) 0)))
                (else
                  (Utilities|Struct|SetmembersinRiftAnomaliesAssetRegistryData (Utilities|Array|Get(aref) _predefinedcircleanomaliesdrops_2 _array_index_2) 0))))))))))

(fn SelectAnomaliesByWeight
  (params (Quantity :type Integer) (AnomalyGroupConfig_WeightsForQuantityOfAnomaliesInAMatch :type Array%20of%20Integers) (AnomalyGroupConfig_RASafeZonesConfig :type Map%20of%20Gameplay%20Tag%20Structures%20to%20Fort%20RAConfig%20Structures))
  (bind (_output_raname _output_ratag) (Utilities|Struct|BreakFortRAChance (Utilities|Array|Get(acopy) (LocalVariables|GetRAWeightedList) (LocalVariables|GetSelectedIndexRA))))
  (bind (_riftanomalysystemconfigdata_perstormcirclegeneralspawnchance_26_aaeceaa646d26e24c4340f8a66ffba3b _riftanomalysystemconfigdata_minsafezone_40_6abc48b2422b44b585ba67a5f74173a6 _riftanomalysystemconfigdata_maxsafezone_41_f34071834b0d86641a5860a838c2db3c _riftanomalysystemconfigdata_predictionwindow_37_4ad963b14b0f959ff7c44898407c2348 _riftanomalysystemconfigdata_cananomaliesrepeat_44_6ce45a574f0d808a542d9fb5e0603294 _riftanomalysystemconfigdata_rasystemenabled_46_fdecd3964c50a8aaab93158bbb40a992 _riftanomalysystemconfigdata_decreaseweightofanomalyinarow_50_39fec66249842b8db450b59845657a63 _riftanomalysystemconfigdata_maxsafezonespermatch_107_9a5f555f4dd71278a52e6b952b5e362a _riftanomalysystemconfigdata_overridestormweightbyphaserowbytag_60_9a5ba2b34503c08e68ce4f841220589b _riftanomalysystemconfigdata_chanceofspawnsafezonebasedanomalies_71_0941dfcb4c5215d20a9246a310499acb _riftanomalysystemconfigdata_anomalysubsetoverriding_70_82b0e46f433a27407a9f93a568955f28 _riftanomalysystemconfigdata_usechancebyanomalysystemmode_84_c460390f49dae179dabf328c947469ec _riftanomalysystemconfigdata_weightsforquantityofanomaliesinamatch_97_7eecd97b46b44ef99cd0db914f6f2dcd _riftanomalysystemconfigdata_weightofananomalytobesortedonweightedsafezones_102_c579e04941f00bb8e72f949d6856433b _riftanomalysystemconfigdata_defaultweightforanomalyselection_106_86131a7644fbc538c8a7c7911131ce71) (Utilities|Struct|BreakRiftAnomaliesAssetRegistrySystemData (Variables|GetRiftAnomalySystemConfigData_PerStormCircleGeneralSpawnChance_26_AAECEAA646D26E24C4340F8A66FFBA3B)))
  (bind _raanomaliesconfig (LocalVariables|GetRAAnomaliesConfig))
  (LocalVariables|SetDefaultAnomalyWeight _riftanomalysystemconfigdata_defaultweightforanomalyselection_106_86131a7644fbc538c8a7c7911131ce71)
  (LocalVariables|SetRAAnomaliesConfig AnomalyGroupConfig_RASafeZonesConfig)
  (for _array_element (Variables|DataRegistry|GetPossibleAnomaliesToSpawn)
    (bind _array_element_anomaliedata_2_f38e3dd0464ff4283aff31b4e94803e4 (Utilities|Struct|BreakBRADataRegistryWithRowNameStruct _array_element))
    (bind (_array_element_anomaliedata_2_f38e3dd0464ff4283aff31b4e94803e4_anomaliesdataassets_20_7e8609f440fac2369e8ac7902781fa79 _array_element_anomaliedata_2_f38e3dd0464ff4283aff31b4e94803e4_stormweights_26_be7c04904bf803d046fa5282e362cc01 _array_element_anomaliedata_2_f38e3dd0464ff4283aff31b4e94803e4_canoccurmultipletimes_16_53ee0f1f44ebe0d30c76e08a31e8c644 _array_element_anomaliedata_2_f38e3dd0464ff4283aff31b4e94803e4_effectradius_24_aaeceaa646d26e24c4340f8a66ffba3b _array_element_anomaliedata_2_f38e3dd0464ff4283aff31b4e94803e4_isenabled_30_aac59cb249d32975a549d19e98331703 _array_element_anomaliedata_2_f38e3dd0464ff4283aff31b4e94803e4_playlisttagquery_38_760275694a7508c4a9500d8393a0ca1e _array_element_anomaliedata_2_f38e3dd0464ff4283aff31b4e94803e4_stormweightbyphaserow_43_f8665c0a434e4870bac26abb909a92d3 _array_element_anomaliedata_2_f38e3dd0464ff4283aff31b4e94803e4_ratag_47_3bb6e306408e53230fe68fac1bcd5bb3) (Utilities|Struct|BreakRiftAnomaliesAssetRegistryData _array_element_anomaliedata_2_f38e3dd0464ff4283aff31b4e94803e4))
    (bind _output_get (LocalVariables|SetIteratedAnomalyTag _array_element_anomaliedata_2_f38e3dd0464ff4283aff31b4e94803e4_ratag_47_3bb6e306408e53230fe68fac1bcd5bb3))
    (bind (_value _returnvalue) (Utilities|Map|Find _raanomaliesconfig _output_get))
    (bind (_value_defaultweightifnosafezoneisset _value_raweightpersafezonearray _value_selectionweight) (Utilities|Struct|BreakFortRAConfig _value))
    (if _returnvalue
      (Utilities|Array|Add (LocalVariables|GetRAWeightedList) (Utilities|Struct|MakeFortRAChance (GameplayTags|GetTagName (LocalVariables|GetIteratedAnomalyTag)) (LocalVariables|GetIteratedAnomalyTag) (Math|Integer|SelectInt _value_selectionweight (LocalVariables|GetDefaultAnomalyWeight) _returnvalue)))
      (else
        (Fortnite|FortShippingLog self (Utilities|String|Append "LogFortRAManager: No tag defined for anomaly at this selection group  " (Utilities|GetDisplayName _array_element_anomaliedata_2_f38e3dd0464ff4283aff31b4e94803e4_anomaliesdataassets_20_7e8609f440fac2369e8ac7902781fa79) "Not adding it into the array of this group")))))
  (for _ (range 1 (+ (Variables|GetQuantity) 1))
    (CallFunction|RASelectAnomalyfromWeightedStructure 0 (LocalVariables|GetRAWeightedList) (LocalVariables|GetSelectedIndexRA))
    (if (Utilities|Array|IsValidIndex (LocalVariables|GetRAWeightedList) (LocalVariables|GetSelectedIndexRA))
      (Utilities|Array|Add (LocalVariables|GetRASelectedList) _output_ratag)
      (Utilities|Array|RemoveIndex (LocalVariables|GetRAWeightedList) (LocalVariables|GetSelectedIndexRA))
      (else
        (Fortnite|FortShippingLog self "LogFortRAManager: Something is wrong index should always be valid on this case" true))))
  (return (LocalVariables|GetRASelectedList)))

(fn SelectAnomaliesSafezonesPrioritizingAnomalies
  (params (SelectedAnomalies :type Array%20of%20Gameplay%20Tag%20Structures) (AnomaliesGroupConfig_WeightsForQuantityOfAnomaliesInAMatch :type Array%20of%20Integers) (AnomaliesGroupConfig_RASafeZonesConfig :type Map%20of%20Gameplay%20Tag%20Structures%20to%20Fort%20RAConfig%20Structures))
  (bind _currentiteratedanomalytag (Variables|Default|GetCurrentIteratedAnomalyTag))
  (bind _anomaliesconfigdata (LocalVariables|GetAnomaliesConfigData))
  (bind (_value _returnvalue) (Utilities|Map|Find _anomaliesconfigdata _currentiteratedanomalytag))
  (bind (_value_defaultweightifnosafezoneisset _value_raweightpersafezonearray) (Utilities|Struct|BreakFortRAConfig _value))
  (bind (_riftanomalysystemconfigdata_perstormcirclegeneralspawnchance_26_aaeceaa646d26e24c4340f8a66ffba3b _riftanomalysystemconfigdata_minsafezone_40_6abc48b2422b44b585ba67a5f74173a6 _riftanomalysystemconfigdata_maxsafezone_41_f34071834b0d86641a5860a838c2db3c _riftanomalysystemconfigdata_predictionwindow_37_4ad963b14b0f959ff7c44898407c2348 _riftanomalysystemconfigdata_cananomaliesrepeat_44_6ce45a574f0d808a542d9fb5e0603294 _riftanomalysystemconfigdata_rasystemenabled_46_fdecd3964c50a8aaab93158bbb40a992 _riftanomalysystemconfigdata_decreaseweightofanomalyinarow_50_39fec66249842b8db450b59845657a63 _riftanomalysystemconfigdata_maxsafezonespermatch_107_9a5f555f4dd71278a52e6b952b5e362a) (Utilities|Struct|BreakRiftAnomaliesAssetRegistrySystemData (Variables|GetRiftAnomalySystemConfigData_PerStormCircleGeneralSpawnChance_26_AAECEAA646D26E24C4340F8A66FFBA3B)))
  (bind _selectedsafezoneperanomaly (LocalVariables|GetSelectedSafeZonePerAnomaly))
  (LocalVariables|SetAnomaliesConfigData AnomaliesGroupConfig_RASafeZonesConfig)
  (LocalVariables|SetMaxSafeZonesPerMatch _riftanomalysystemconfigdata_maxsafezonespermatch_107_9a5f555f4dd71278a52e6b952b5e362a)
  (if (Utilities|Array|IsNotEmpty SelectedAnomalies)
    (for _array_element (LocalVariables|GetSelectedAnomalies)
      (for _index (range (+ (LocalVariables|GetMaxSafeZonesPerMatch) 1))
        (LocalVariables|SetCurrentIteratedSafeZoneIndex _index)
        (Variables|Default|SetCurrentIteratedAnomalyTag _array_element)
        ; SafeZoneAlreadyTaken
        (bind _safezoneindexfree (CallFunction|IsSafeZoneFree (LocalVariables|GetCurrentIteratedSafeZoneIndex)))
        (if _safezoneindexfree
          (if _returnvalue
            (if (Utilities|Array|IsNotEmpty _value_raweightpersafezonearray)
              (if (Utilities|Array|IsValidIndex _value_raweightpersafezonearray (LocalVariables|GetCurrentIteratedSafeZoneIndex))
                (if (not (Utilities|Array|ContainsItem (LocalVariables|GetAlreadySelectedSafeZones) (LocalVariables|GetCurrentIteratedSafeZoneIndex)))
                  (Utilities|Array|Add (LocalVariables|GetSafeZoneWeightsList) (Utilities|Array|Get(acopy) _value_raweightpersafezonearray (LocalVariables|GetCurrentIteratedSafeZoneIndex)))
                  (else
                    (Utilities|Array|Add (LocalVariables|GetSafeZoneWeightsList))))
                (elif (and (Utilities|Array|IsValidIndex (Variables|Default|GetRift%20Anomaly%20System%20Config%20Data%20Default%20Safe%20Zone%20Weights) (LocalVariables|GetCurrentIteratedSafeZoneIndex)) (not (Utilities|Array|ContainsItem (LocalVariables|GetAlreadySelectedSafeZones) (LocalVariables|GetCurrentIteratedSafeZoneIndex))))
                  ; Grabbing the default value
                  (Utilities|Array|Add (LocalVariables|GetSafeZoneWeightsList) (Utilities|Array|Get(acopy) (Variables|Default|GetRift%20Anomaly%20System%20Config%20Data%20Default%20Safe%20Zone%20Weights) (LocalVariables|GetCurrentIteratedSafeZoneIndex)))
                  (else
                    (Utilities|Array|Add (LocalVariables|GetSafeZoneWeightsList))))))
            (else
              _))
          (else
            (Utilities|Array|Add (LocalVariables|GetSafeZoneWeightsList)))))
      ; Will select the anomaly that should spawn on the safe zone according to weight
      (CallFunction|RASelectIndexfromWeightedStructure (LocalVariables|GetSafeZoneWeightsList) _selectedsafezoneperanomaly)
      (if (> _selectedsafezoneperanomaly 0)
        (Utilities|Map|Add (LocalVariables|GetSelectedAnomaliesBySafeZoneIndex) (LocalVariables|GetSelectedSafeZonePerAnomaly) (Variables|Default|GetCurrentIteratedAnomalyTag))
        (Utilities|Array|Clear (LocalVariables|GetSafeZoneWeightsList))
        (Utilities|Array|Add (LocalVariables|GetAlreadySelectedSafeZones) (LocalVariables|GetSelectedSafeZonePerAnomaly))
        (else
          (Utilities|Array|Clear (LocalVariables|GetSafeZoneWeightsList)))))
    (return (LocalVariables|GetSelectedAnomaliesBySafeZoneIndex))))

(fn SelectedAnomaliesSafeZonesBasedOnWeight
  (params (SelectedAnomalies :type Array%20of%20Gameplay%20Tag%20Structures))
  (bind _anomaliesconfigdata (LocalVariables|GetAnomaliesConfigData))
  (bind (_riftanomalysystemconfigdata_perstormcirclegeneralspawnchance_26_aaeceaa646d26e24c4340f8a66ffba3b _riftanomalysystemconfigdata_minsafezone_40_6abc48b2422b44b585ba67a5f74173a6 _riftanomalysystemconfigdata_maxsafezone_41_f34071834b0d86641a5860a838c2db3c _riftanomalysystemconfigdata_predictionwindow_37_4ad963b14b0f959ff7c44898407c2348 _riftanomalysystemconfigdata_cananomaliesrepeat_44_6ce45a574f0d808a542d9fb5e0603294 _riftanomalysystemconfigdata_rasystemenabled_46_fdecd3964c50a8aaab93158bbb40a992 _riftanomalysystemconfigdata_decreaseweightofanomalyinarow_50_39fec66249842b8db450b59845657a63 _riftanomalysystemconfigdata_maxsafezonespermatch_107_9a5f555f4dd71278a52e6b952b5e362a) (Utilities|Struct|BreakRiftAnomaliesAssetRegistrySystemData (Variables|GetRiftAnomalySystemConfigData_PerStormCircleGeneralSpawnChance_26_AAECEAA646D26E24C4340F8A66FFBA3B)))
  (bind (_riftanomalysystemconfigdata_perstormcirclegeneralspawnchance_26_aaeceaa646d26e24c4340f8a66ffba3b_2 _riftanomalysystemconfigdata_minsafezone_40_6abc48b2422b44b585ba67a5f74173a6_2 _riftanomalysystemconfigdata_maxsafezone_41_f34071834b0d86641a5860a838c2db3c_2 _riftanomalysystemconfigdata_predictionwindow_37_4ad963b14b0f959ff7c44898407c2348_2 _riftanomalysystemconfigdata_cananomaliesrepeat_44_6ce45a574f0d808a542d9fb5e0603294_2 _riftanomalysystemconfigdata_rasystemenabled_46_fdecd3964c50a8aaab93158bbb40a992_2 _riftanomalysystemconfigdata_decreaseweightofanomalyinarow_50_39fec66249842b8db450b59845657a63_2 _riftanomalysystemconfigdata_maxsafezonespermatch_107_9a5f555f4dd71278a52e6b952b5e362a_2 _riftanomalysystemconfigdata_overridestormweightbyphaserowbytag_60_9a5ba2b34503c08e68ce4f841220589b _riftanomalysystemconfigdata_chanceofspawnsafezonebasedanomalies_71_0941dfcb4c5215d20a9246a310499acb _riftanomalysystemconfigdata_anomalysubsetoverriding_70_82b0e46f433a27407a9f93a568955f28 _riftanomalysystemconfigdata_usechancebyanomalysystemmode_84_c460390f49dae179dabf328c947469ec _riftanomalysystemconfigdata_weightsforquantityofanomaliesinamatch_97_7eecd97b46b44ef99cd0db914f6f2dcd _riftanomalysystemconfigdata_weightofananomalytobesortedonweightedsafezones_102_c579e04941f00bb8e72f949d6856433b) (Utilities|Struct|BreakRiftAnomaliesAssetRegistrySystemData (Variables|GetRiftAnomalySystemConfigData_PerStormCircleGeneralSpawnChance_26_AAECEAA646D26E24C4340F8A66FFBA3B)))
  (LocalVariables|SetAnomaliesConfigData _riftanomalysystemconfigdata_weightofananomalytobesortedonweightedsafezones_102_c579e04941f00bb8e72f949d6856433b)
  (LocalVariables|SetMaxSafeZonesPerMatch _riftanomalysystemconfigdata_maxsafezonespermatch_107_9a5f555f4dd71278a52e6b952b5e362a)
  (if (Utilities|Array|IsNotEmpty SelectedAnomalies)
    (for _index (range (+ (LocalVariables|GetMaxSafeZonesPerMatch) 1))
      (LocalVariables|SetCurrentIteratedSafeZoneIndex _index)
      (for _array_element (Variables|GetSelectedAnomalies)
        (bind _output_get (Variables|Default|SetCurrentIteratedAnomalyTag _array_element))
        (bind (_value _returnvalue) (Utilities|Map|Find _anomaliesconfigdata _output_get))
        (bind (_value_defaultweightifnosafezoneisset _value_raweightpersafezonearray) (Utilities|Struct|BreakFortRAConfig _value))
        (if _returnvalue
          (if (Utilities|Array|IsNotEmpty _value_raweightpersafezonearray)
            (if (Utilities|Array|IsValidIndex _value_raweightpersafezonearray (LocalVariables|GetCurrentIteratedSafeZoneIndex))
              (Utilities|Array|Add (LocalVariables|GetSafeZoneWeightsList) (Utilities|Array|Get(acopy) _value_raweightpersafezonearray (LocalVariables|GetCurrentIteratedSafeZoneIndex)))
              (elif (Utilities|Array|IsValidIndex (Variables|Default|GetRift%20Anomaly%20System%20Config%20Data%20Default%20Safe%20Zone%20Weights) (LocalVariables|GetCurrentIteratedSafeZoneIndex))
                ; Grabbing the default value
                (Utilities|Array|Add (LocalVariables|GetSafeZoneWeightsList) (Utilities|Array|Get(acopy) (Variables|Default|GetRift%20Anomaly%20System%20Config%20Data%20Default%20Safe%20Zone%20Weights) (LocalVariables|GetCurrentIteratedSafeZoneIndex)))
                (else
                  (Utilities|Array|Add (LocalVariables|GetSafeZoneWeightsList))))))))
      ; Will select the anomaly that should spawn on the safe zone according to weight
      (CallFunction|RASelectIndexfromWeightedStructure (LocalVariables|GetSafeZoneWeightsList) (LocalVariables|GetSelectedAnomalyIndex))
      (Utilities|Array|Add (LocalVariables|GetSelectedAnomaliesBySafeZoneIndex) (Utilities|Array|Get(acopy) (Variables|GetSelectedAnomalies) (LocalVariables|GetSelectedAnomalyIndex)))
      (Utilities|Array|RemoveIndex (Variables|GetSelectedAnomalies) (LocalVariables|GetSelectedAnomalyIndex))
      (Utilities|Array|Clear (LocalVariables|GetSafeZoneWeightsList)))
    (return (LocalVariables|GetSelectedAnomaliesBySafeZoneIndex))))

(fn SetMatchQuantityOfAnomalies
  (params (QuantityWeightsArray :type Array%20of%20Integers))
  (bind _selectedindexasquantity (LocalVariables|GetSelectedIndexAsQuantity))
  (if (Utilities|Array|IsNotEmpty QuantityWeightsArray)
    (CallFunction|RASelectIndexfromWeightedStructure (Variables|GetQuantityWeightsArray) _selectedindexasquantity)
    (Fortnite|FortShippingLog self (Utilities|String|Append "LogFortRAManager: MaxQuantityOfAnomaliesInThisMatchIs: " (Utilities|String|ToString(Integer) _selectedindexasquantity)))
    (return (LocalVariables|GetSelectedIndexAsQuantity))
    (else
      (Fortnite|FortShippingLog self "LogFortRAManager: No value defined on RiftAnomaliesQuantityArray, Anomaly Quantity will be set to 0" true)
      (return 0))))

(fn SetRASafeZoneSpawnState
  (params (SystemResumedNewValue :type Boolean))
  (Networking|SwitchHasAuthority
    (:Authority
      (Variables|Default|SetbSafeZoneSpawnActive SystemResumedNewValue))))

(fn SetUIState
  (params (NewState :type BP_Enum_RA_UI_STATE%20Enum :default "NewEnumerator0"))
  (switch Utilities|FlowControl|Switch|SwitchonBP_Enum_RA_UI_STATE NewState
    (:NewEnumerator1
      (switch Utilities|FlowControl|Switch|SwitchonBP_Enum_RA_UI_STATE (Variables|ui|GetRiftAnomaliesUIState)
        (:NewEnumerator0
          (Variables|ui|SetRiftAnomaliesUIState "NewEnumerator1"))
        (:NewEnumerator4
          _)
        (:NewEnumerator1
          _))
      (UI|CheckifCanActivateStartMatchAnomalyUI))))

(fn SetupAnticipation
  (bind (_riftanomalysystemconfigdata_perstormcirclegeneralspawnchance_26_aaeceaa646d26e24c4340f8a66ffba3b _riftanomalysystemconfigdata_minsafezone_40_6abc48b2422b44b585ba67a5f74173a6 _riftanomalysystemconfigdata_maxsafezone_41_f34071834b0d86641a5860a838c2db3c _riftanomalysystemconfigdata_predictionwindow_37_4ad963b14b0f959ff7c44898407c2348) (Utilities|Struct|BreakRiftAnomaliesAssetRegistrySystemData (Variables|GetRiftAnomalySystemConfigData_PerStormCircleGeneralSpawnChance_26_AAECEAA646D26E24C4340F8A66FFBA3B)))
  (Utilities|Time|ClearandInvalidateTimerbyHandle (Variables|Timers|GetSpawnPrediction_TimerHandle))
  (LocalVariables|SetTimeTIllSpawnAnticipation (- (Utilities|Struct|BreakFortSafeZonePhaseInfo (Variables|Default|GetCurrentSafeZoneInfo)) _riftanomalysystemconfigdata_predictionwindow_37_4ad963b14b0f959ff7c44898407c2348))
  (if (<= (LocalVariables|GetTimeTIllSpawnAnticipation) 0)
    (Prediction|SpawnAnticipation)
    (else
      (bind _returnvalue (Utilities|Time|SetTimerbyEvent (EventDispatchers|CreateEvent :!event SpawnAnticipation) (LocalVariables|GetTimeTIllSpawnAnticipation)))
      (Variables|Timers|SetSpawnPrediction_TimerHandle _returnvalue))))

(fn SetupNextPhaseRA
  (bind _output (Utilities|Array|Get(acopy) (Variables|DataRegistry|GetPredefinedCircleAnomaliesDrops) (+ (SafeZone|GetCurrentPhaseIndex (SafeZone|GetSafeZoneIndicator)) 1)))
  (bind _anomaliesdataassets_20_7e8609f440fac2369e8ac7902781fa79 (Utilities|Struct|BreakRiftAnomaliesAssetRegistryData _output))
  (bind _returnvalue (SafeZone|GetSafeZoneIndicator))
  (bind _returnvalue_2 (SafeZone|GetCurrentPhaseIndex _returnvalue))
  (bind _retryinganomaly (LocalVariables|GetRetryingAnomaly))
  (bind _returnvalue_3 (Utilities|IsValidSoftObjectReference _retryinganomaly))
  (bind _returnvalue_4 (CallFunction|GetAnomalyTryingtoRetry))
  (bind _output_2 (Utilities|ResolveSoftReference _anomaliesdataassets_20_7e8609f440fac2369e8ac7902781fa79))
  (bind _returnvalue_5 (select _returnvalue_3 _retryinganomaly _output_2))
  (bind _output_3 (Utilities|ResolveSoftReference _returnvalue_5))
  (bind _returnvalue_6 (+ _returnvalue_2 1))
  (bind _predefinedcircleanomaliesdrops (Variables|DataRegistry|GetPredefinedCircleAnomaliesDrops))
  (bind _currentsafezoneinfo_center (Utilities|Struct|BreakFortSafeZonePhaseInfo (Variables|GetCurrentSafeZoneInfo_Center)))
  (if (Utilities|IsValidSoftObjectReference _returnvalue_4)
    (LocalVariables|SetRetryingAnomaly _returnvalue_4)
    (elif (Utilities|Array|IsValidIndex _predefinedcircleanomaliesdrops _returnvalue_6)
      _))
  (Utilities|IsValid _output_3
    (:Is%20Valid
      (bind _returnvalue_7 (CallFunction|ExecuteAnomalyCanSpawnRule _returnvalue_5 (LocalVariables|GetIsRetrying)))
      (Debugging|PrintInteger (RiftAnomaly|Spawn|GetMaxRetries _output_3 :!class /Script/RiftAnomaliesGameplayRuntime.FortRiftAnomaliesData)
        (:Then
          (if _returnvalue_7
            (bind _output_get (Variables|Default|SetIncomingRA (Utilities|ResolveSoftReference _returnvalue_5)))
            (if (Class|FortRiftAnomaliesData|GetIsUsingUseEQS _output_get)
              (CallFunction|PerformEQS)
              (else
                (bind _output_get_2 (Variables|Default|SetSpawnLocation _currentsafezoneinfo_center))
                (Prediction|SetupAnticipation)
                (CallFunction|AddAnomalyInfo (Utilities|String|ToString(Text) (Class|FortRiftAnomaliesData|GetRAName (Variables|Default|GetIncomingRA) :!class /Script/RiftAnomaliesGameplayRuntime.FortRiftAnomaliesData)) _output_get_2)))))))
    (:Is%20Not%20Valid
      (Development|PrintString "This Storm Circle Doesn't has an valid anomaly"))))

(fn ShouldAnomalyStayInThePool
  (params (RAData_AnomaliesDataAssets_20_7E8609F440FAC2369E8AC7902781FA79 :type Fort%20Rift%20Anomalies%20Data%20Object%20Reference) (RAData_StormWeights_26_BE7C04904BF803D046FA5282E362CC01 :type Array%20of%20Integers) (RAData_CanOccurMultipleTimes_16_53EE0F1F44EBE0D30C76E08A31E8C644 :type Boolean) (RAData_EffectRadius_24_AAECEAA646D26E24C4340F8A66FFBA3B :type Float%20%28double-precision%29) (RAData_IsEnabled_30_AAC59CB249D32975A549D19E98331703 :type Boolean) (RAData_PlaylistTagQuery_38_760275694A7508C4A9500D8393A0CA1E :type Gameplay%20Tag%20Query%20Structure) (RAData_StormWeightByPhaseRow_43_F8665C0A434E4870BAC26ABB909A92D3 :type Scalable%20Float%20Structure) (RAData_RATag_47_3BB6E306408E53230FE68FAC1BCD5BB3 :type Gameplay%20Tag%20Structure) (RAData_RAModfierMap_59_87219A6747987167EB1CC3945BD39CF7 :type Map%20of%20Gameplay%20Tag%20Structures%20to%20Fort%20RAWeight%20Modifier%20Structures) (RAData_LastAnomaly_62_78D3C0494913613382169FB5D22A26D2 :type Boolean) (RAData_ForceShowUI_64_3859B8CF4298278ED0F3329E21080F62 :type Boolean))
  (if RAData_IsEnabled_30_AAC59CB249D32975A549D19E98331703
    (if (GameplayTags|IsTagQueryEmpty RAData_PlaylistTagQuery_38_760275694A7508C4A9500D8393A0CA1E)
      (return true)
      (else
        (bind _output_get (LocalVariables|SetTagQuery RAData_PlaylistTagQuery_38_760275694A7508C4A9500D8393A0CA1E))
        (bind _output_get_2 (LocalVariables|SetTagQueryResult (GameplayTags|DoesContainerMatchTagQuery (Class|FortGameStateAthena|GetAthenaPlaylistContextTags (Fortnite|Athena|GameState|GetGameStateAthena)) _output_get)))
        (return _output_get_2)))
    (else
      (Development|PrintString (Utilities|String|Append (Utilities|String|ToString(Text) (Class|FortRiftAnomaliesData|GetRAName RAData_AnomaliesDataAssets_20_7E8609F440FAC2369E8AC7902781FA79 :!class /Script/RiftAnomaliesGameplayRuntime.FortRiftAnomaliesData)) " is disabled"))
      (return false))))

(fn SpawnAnomalyFX
  (bind _returnvalue (SafeZone|GetSafeZoneIndicator))
  (bind _anomalyfxclass (Variables|Default|GetAnomalyFXClass))
  (bind _fxindicator (Variables|Default|GetFXIndicator))
  (Utilities|IsValid _returnvalue
    (:Is%20Valid
      (Utilities|IsValid _fxindicator
        (:Is%20Valid
          (Actor|DestroyActor _fxindicator))
        (:Is%20Not%20Valid
          _))
      (if (Utilities|IsValidClass _anomalyfxclass)
        (bind _returnvalue_2 (Game|SpawnActorfromClass _anomalyfxclass (Math|Transform|MakeTransform (SafeZone|GetSafeZoneCenter _returnvalue)) "AlwaysSpawn" "MultiplyWithRoot" 0 0))
        (Variables|Default|SetFXIndicator _returnvalue_2)))))

(fn SpawnAnticipation
  (CallFunction|UIBroadcastIncomingRA (Variables|Default|GetIncomingRA)))

(fn SystemOnStormCircleWeightedMode
  (Determinism|PreDefiningAllCircleAnomaliesonSafeZoneWeightedMode))

(fn SystemOnWeightedAnomalySelectionMode
  (if (Determinism|IsGoingtoSpawnAnySafeZoneBasedRAs)
    (Determinism|PreDefiningAllCircleAnomaliesonAnomalyWeightedMode)))

(fn TrySpawnRiftAnomaly
  (params (Location :type Vector) (Data :type Fort%20Rift%20Anomalies%20Data%20Object%20Reference%20%28by%20ref%29))
  (bind (_riftanomalysystemconfigdata_perstormcirclegeneralspawnchance_26_aaeceaa646d26e24c4340f8a66ffba3b _riftanomalysystemconfigdata_minsafezone_40_6abc48b2422b44b585ba67a5f74173a6 _riftanomalysystemconfigdata_maxsafezone_41_f34071834b0d86641a5860a838c2db3c _riftanomalysystemconfigdata_predictionwindow_37_4ad963b14b0f959ff7c44898407c2348 _riftanomalysystemconfigdata_cananomaliesrepeat_44_6ce45a574f0d808a542d9fb5e0603294 _riftanomalysystemconfigdata_rasystemenabled_46_fdecd3964c50a8aaab93158bbb40a992 _riftanomalysystemconfigdata_decreaseweightofanomalyinarow_50_39fec66249842b8db450b59845657a63 _riftanomalysystemconfigdata_maxsafezonespermatch_107_9a5f555f4dd71278a52e6b952b5e362a _riftanomalysystemconfigdata_overridestormweightbyphaserowbytag_60_9a5ba2b34503c08e68ce4f841220589b _riftanomalysystemconfigdata_chanceofspawnsafezonebasedanomalies_71_0941dfcb4c5215d20a9246a310499acb _riftanomalysystemconfigdata_anomalysubsetoverriding_70_82b0e46f433a27407a9f93a568955f28 _riftanomalysystemconfigdata_usechancebyanomalysystemmode_84_c460390f49dae179dabf328c947469ec _riftanomalysystemconfigdata_weightsforquantityofanomaliesinamatch_97_7eecd97b46b44ef99cd0db914f6f2dcd _riftanomalysystemconfigdata_weightofananomalytobesortedonweightedsafezones_102_c579e04941f00bb8e72f949d6856433b _riftanomalysystemconfigdata_defaultweightforanomalyselection_106_86131a7644fbc538c8a7c7911131ce71 _riftanomalysystemconfigdata_defaultsafezoneweights_110_d4f62c4d4b770b30b08cf3acd19dd541 _riftanomalysystemconfigdata_prioritizeanomalywhiledecidingsafezones_114_9dbf0cf64512cfb0f4864b9ecdd77549 _riftanomalysystemconfigdata_anomaliesgroupsconfiguration_118_57b7639f457cde58dce1cfa3ccb50010 _riftanomalysystemconfigdata_uinotificationsenabled_121_4cd43fbe4608f5746aa5908accc4cdd6) (Utilities|Struct|BreakRiftAnomaliesAssetRegistrySystemData (Variables|GetRiftAnomalySystemConfigData_PerStormCircleGeneralSpawnChance_26_AAECEAA646D26E24C4340F8A66FFBA3B)))
  (Utilities|IsValid Data
    (:Is%20Valid
      (bind _anomalydataregistrydata (DataRegistry|FindRegistrybyDataAsset Data))
      (bind (_anomalydataregistrydata_anomaliesdataassets_20_7e8609f440fac2369e8ac7902781fa79 _anomalydataregistrydata_stormweights_26_be7c04904bf803d046fa5282e362cc01 _anomalydataregistrydata_canoccurmultipletimes_16_53ee0f1f44ebe0d30c76e08a31e8c644 _anomalydataregistrydata_effectradius_24_aaeceaa646d26e24c4340f8a66ffba3b _anomalydataregistrydata_isenabled_30_aac59cb249d32975a549d19e98331703 _anomalydataregistrydata_playlisttagquery_38_760275694a7508c4a9500d8393a0ca1e _anomalydataregistrydata_stormweightbyphaserow_43_f8665c0a434e4870bac26abb909a92d3 _anomalydataregistrydata_ratag_47_3bb6e306408e53230fe68fac1bcd5bb3 _anomalydataregistrydata_ramodfiermap_59_87219a6747987167eb1cc3945bd39cf7 _anomalydataregistrydata_lastanomaly_62_78d3c0494913613382169fb5d22a26d2 _anomalydataregistrydata_forceshowui_64_3859b8cf4298278ed0f3329e21080f62) (Utilities|Struct|BreakRiftAnomaliesAssetRegistryData _anomalydataregistrydata))
      (LocalVariables|SetForceShowUI _anomalydataregistrydata_forceshowui_64_3859b8cf4298278ed0f3329e21080f62)
      (bind _output_get (Variables|Default|SetLastAnomalyIncoming _anomalydataregistrydata_lastanomaly_62_78d3c0494913613382169fb5d22a26d2))
      (CallFunction|SetRASafeZoneSpawnState (Widget|UEFN|InvertBool _output_get))
      (if (Utilities|IsValidClass (Class|FortRiftAnomaliesData|GetRiftAnomalieActorClass Data))
        (bind _returnvalue (Game|SpawnActorfromClass (Class|FortRiftAnomaliesData|GetRiftAnomalieActorClass (Variables|GetData)) (Math|Transform|MakeTransform (Variables|GetLocation)) "AlwaysSpawn" "MultiplyWithRoot" 0 (Variables|GetData) 0))
        (bind _output_get_2 (LocalVariables|SetSpawnedRAActorBase _returnvalue))
        (Debugging|PrintVector (Transformation|GetActorLocation _output_get_2) (Utilities|String|ToString(Text) (Utilities|Text|FormatText "NSLOCTEXT(\"[4B9D47DCB44E2CC832FF8DB0E2DA666C]\", \"DE80E93B439971323D32F19298B2440F\", \"RA \\\"{a}\\\" successfully spawned at {b}\")" (Variables|/Script/RiftAnomaliesGameplayRuntime.FortRiftAnomaliesData|GetRAName (Variables|GetData)) (Utilities|String|ToString(Vector) (Transformation|GetActorLocation _output_get_2)))) 12.0 "(R=0.076514,G=1.000000,B=0.001511,A=1.000000)"
          (:Then
            _)))
      (Utilities|Array|AddUnique (Variables|Default|GetAllActiveRAInfo) (Utilities|Struct|MakeBRAActiveInfoStruct (Variables|GetData) (LocalVariables|GetSpawnedRAActorBase)))
      (bind _output_get_3 (Variables|Default|SetActiveRA (Variables|GetData)))
      (CallFunction|DoAnomalyTypeActions _output_get_3)
      (switch Utilities|FlowControl|Switch|SwitchonERAEffectsRules (Class|FortRiftAnomaliesData|GetEffectApplicationRules (Variables|GetData))
        (:DontApplyEffects
          _)
        (:ApplyToAllPlayers
          (bind _outfortplayerpawns (Fortnite|GetFortPlayerPawns))
          (for _array_element _outfortplayerpawns
            (Utilities|IsValid _array_element
              (:Is%20Valid
                (CallFunction|ApplyRAEffects _array_element (Variables|GetData)))))
          (CallFunction|ListenforPawnSpawnforDeadPlayers))
        (:ApplyViaOverlap
          _))
      (if (or (LocalVariables|GetForceShowUI) _riftanomalysystemconfigdata_uinotificationsenabled_121_4cd43fbe4608f5746aa5908accc4cdd6)
        (CallFunction|UIBroadcastRAActivated (Variables|Default|GetActiveRA))
        (else
          _))
      (CallFunction|SoundBroadcastRAPlayTriggerSound (Class|FortRiftAnomaliesData|GetRiftAnomalyTriggerSound (Variables|Default|GetActiveRA)))
      (Variables|Default|SetIncomingRA 0)
      (Spawning|SpawnAnomalyFX))))

(fn TryToActivateFullMatchAnomaly
  (params (AnomalyToCheck :type Fort%20Rift%20Anomalies%20Data%20Object%20Reference))
  (switch Utilities|FlowControl|Switch|SwitchonERABehavior (Class|FortRiftAnomaliesData|GetRABehaviorType AnomalyToCheck)
    (:MatchStartOrientedRA
      ;[1 1 1] Spawn And Handle spawn stuff
      (Spawning|TrySpawnRiftAnomaly "0, 0, 0" AnomalyToCheck))))

(fn UnbindSpawnPawnListeners
  ;[1 0 0 0.5] TODO: We need to accommodate bots
  (for _array_element (Fortnite|GetAllFortPlayerControllers true)
    (CallFunction|UnbindfromPlayerPawnSpawn _array_element)))
