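; Cheat manager extension for Rift Anomalies that provides debug commands to spawn anomalies and Caval crates and to toggle safe zone spawning.
; Pause and resume functions find the rift anomalies component on the Athena game state and set its safe zone spawn state.
; Spawn entry points build "Cheat ke*" console commands that invoke ForceSpawnRA or ForceSpawnCrate by name.
; The crate command appends coordinates in X, Z, Y order.
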(event EventOnSpawnRiftAnomalyTriggered
  (params (RADebugName :type String))
  (Development|ExecuteConsoleCommand (Utilities|String|Append "Cheat ke* ForceSpawnRA " RADebugName)))

(fn RiftAnomaliesPauseSafeZoneSpawn
  (bind _returnvalue (Fortnite|Athena|GameState|GetGameStateAthena))
  (bind _returnvalue_2 (Actor|GetComponentbyClass _returnvalue "/RiftAnomaliesGameplay/Gameplay/B_RiftAnomaliesGameStateComponent.B_RiftAnomaliesGameStateComponent_C"))
  (Utilities|IsValid _returnvalue_2
    (:Is%20Valid
      (Class|BRiftAnomaliesGameStateComponent|SetRASafeZoneSpawnState _returnvalue_2 false))))

(fn RiftAnomaliesResumeSafeZoneSpawn
  (bind _returnvalue (Fortnite|Athena|GameState|GetGameStateAthena))
  (bind _returnvalue_2 (Actor|GetComponentbyClass _returnvalue "/RiftAnomaliesGameplay/Gameplay/B_RiftAnomaliesGameStateComponent.B_RiftAnomaliesGameStateComponent_C"))
  (Utilities|IsValid _returnvalue_2
    (:Is%20Valid
      (Class|BRiftAnomaliesGameStateComponent|SetRASafeZoneSpawnState _returnvalue_2 true))))

(fn SpawnCavalCrateOnPosition
  (params (X :type Float%20%28double-precision%29) (Y :type Float%20%28double-precision%29) (Z :type Float%20%28double-precision%29))
  (Development|ExecuteConsoleCommand (Utilities|String|Append "Cheat ke* ForceSpawnCrate " (Utilities|String|ToString(Float) X) " " (Utilities|String|ToString(Float) Z) " " (Utilities|String|ToString(Float) Y))))
